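# Control Rig for the F_MED_AgentSherbert character body that applies pose-driven offsets to jacket and right-collar bones.
# Spherical pose readers on arms, thighs and spine output weights for relative transform offsets on dyn_ bones, mostly chain bases, run through nested sequences.
# One collar correction multiplies right upper-arm and left forearm readers, so both poses must match.
# Three side-jacket Modify Transforms nodes are unconnected to execution, so never run.
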ControlRigBlueprint /BRCosmetics/Characters/Player/Female/Medium/Bodies/F_MED_AgentSherbert/Meshes/F_MED_AgentSherbert_CtrlRig
graphs [
 RigVMFunctionLibrary {links [], nodes []}
 RigVMModel {nodes [
   RigUnit_BeginExecution = &RigUnit_BeginExecution()
   RigUnit_SequenceAggregate = &RigVMFunction_Sequence(ExecuteContext=@RigUnit_BeginExecution.ExecutePin)
   n1 = &RigUnit_SphericalPoseReader(
    ActiveRegionScaleFactors=(PositiveWidth=1,NegativeWidth=1,PositiveHeight=1,NegativeHeight=1)
    ActiveRegionSize=0.1
    Debug=(bDrawDebug=false,bDraw2D=false,bDrawLocalAxes=false,DebugScale=30,DebugSegments=30,DebugThickness=0.1)
    DriverAxis=(X=-0.5,Y=0,Z=0)
    DriverItem=(Type=Bone,Name="upperarm_r")
    FalloffRegionScaleFactors=(PositiveWidth=1,NegativeWidth=1,PositiveHeight=2,NegativeHeight=1)
    FalloffSize=0.34
    OptionalParentItem=(Type=Bone,Name="None")
    RotationOffset=(X=-120,Y=40,Z=0)
    ExecutePin=@RigUnit_SequenceAggregate.A
   )
   > /Script/ControlRig.RigUnit_SetRelativeTransformForItem(
    Child=(Type=Bone,Name="dyn_R_collar_Bace_1")
    Parent=(Type=Bone,Name="dyn_R_collar_Bace_1")
    bParentInitial=False
    bPropagateToChildren=True
    Weight=@n1.OutputParam
    Value=/Script/RigVM.RigVMFunction_MathTransformFromSRT(Location=(X=5,Y=0,Z=0), Rotation=(X=0,Y=0,Z=60), Scale=(X=1,Y=1,Z=1)) {is=struct RigVMUnitNode}.Transform
   ) {
    is=struct RigVMUnitNode
    template="Set Relative Transform::Execute(in Child,in Parent,in bParentInitial,in Value,in Weight,in bPropagateToChildren)"
   }
   > &RigUnit_SetRelativeTransformForItem(
    Child=(Type=Bone,Name="dyn_R_collar_Bace_2")
    Parent=(Type=Bone,Name="dyn_R_collar_Bace_2")
    bParentInitial=False
    bPropagateToChildren=True
    Value=&RigVMFunction_MathTransformFromSRT(Location=(X=5,Y=-5,Z=0), Rotation=(X=0,Y=0,Z=500), Scale=(X=1,Y=1,Z=1)).Transform
    Weight=@n1.OutputParam
   ) {template="Set Relative Transform::Execute(in Child,in Parent,in bParentInitial,in Value,in Weight,in bPropagateToChildren)"}
   > n2 = &RigUnit_SphericalPoseReader(
    ActiveRegionScaleFactors=(PositiveWidth=1.6,NegativeWidth=1,PositiveHeight=1,NegativeHeight=1.6)
    ActiveRegionSize=0.03
    Debug=(bDrawDebug=false,bDraw2D=false,bDrawLocalAxes=false,DebugScale=30,DebugSegments=30,DebugThickness=0.1)
    DriverAxis=(X=-0.5,Y=0,Z=0)
    DriverItem=(Type=Bone,Name="upperarm_r")
    FalloffRegionScaleFactors=(PositiveWidth=1,NegativeWidth=1,PositiveHeight=1,NegativeHeight=1)
    FalloffSize=0.15
    OptionalParentItem=(Type=Bone,Name="None")
    RotationOffset=(X=-70,Y=0,Z=5)
   )
   > n3 = &RigUnit_SphericalPoseReader(
    ActiveRegionScaleFactors=(PositiveWidth=2.2,NegativeWidth=1,PositiveHeight=1,NegativeHeight=1)
    ActiveRegionSize=0.025
    Debug=(bDrawDebug=false,bDraw2D=false,bDrawLocalAxes=false,DebugScale=30,DebugSegments=30,DebugThickness=0.2)
    DriverAxis=(X=0,Y=1,Z=0)
    DriverItem=(Type=Bone,Name="lowerarm_l")
    FalloffRegionScaleFactors=(PositiveWidth=3,NegativeWidth=1,PositiveHeight=1,NegativeHeight=0.5)
    FalloffSize=0.07
    OptionalParentItem=(Type=Bone,Name="None")
    RotationOffset=(X=92,Y=0,Z=60)
   )
   n4 = &RigVMFunction_MathDoubleMul(A=@n3.OutputParam, B=@n2.OutputParam)
   > &RigUnit_SetRelativeTransformForItem(
    Child=(Type=Bone,Name="dyn_R_collar_Bace_1")
    Parent=(Type=Bone,Name="dyn_R_collar_Bace_1")
    bParentInitial=False
    bPropagateToChildren=True
    Weight=@n4.Result
    Value=&RigVMFunction_MathTransformFromSRT(Location=(X=0,Y=-4,Z=0), Rotation=(X=0,Y=0,Z=0), Scale=(X=1,Y=1,Z=1)).Transform
   ) {template="Set Relative Transform::Execute(in Child,in Parent,in bParentInitial,in Value,in Weight,in bPropagateToChildren)"}
   > &RigUnit_SetRelativeTransformForItem(
    Child=(Type=Bone,Name="dyn_R_collar_Bace_2")
    Parent=(Type=Bone,Name="dyn_R_collar_Bace_2")
    bParentInitial=False
    bPropagateToChildren=True
    Value=&RigVMFunction_MathTransformFromSRT(Location=(X=0,Y=-8,Z=0), Rotation=(X=0,Y=0,Z=0), Scale=(X=1,Y=1,Z=1)).Transform
    Weight=@n4.Result
   ) {template="Set Relative Transform::Execute(in Child,in Parent,in bParentInitial,in Value,in Weight,in bPropagateToChildren)"}
   > n5 = &RigUnit_SphericalPoseReader(
    ActiveRegionScaleFactors=(PositiveWidth=1,NegativeWidth=1,PositiveHeight=0.3,NegativeHeight=1)
    ActiveRegionSize=0.1
    Debug=(bDrawDebug=false,bDraw2D=false,bDrawLocalAxes=false,DebugScale=30,DebugSegments=30,DebugThickness=0.1)
    DriverAxis=(X=-0.5,Y=0,Z=0)
    DriverItem=(Type=Bone,Name="upperarm_r")
    FalloffRegionScaleFactors=(PositiveWidth=1,NegativeWidth=1,PositiveHeight=2,NegativeHeight=1)
    FalloffSize=0.3
    OptionalParentItem=(Type=Bone,Name="None")
    RotationOffset=(X=-120,Y=40,Z=0)
   )
   > &RigUnit_SetRelativeTransformForItem(
    Child=(Type=Bone,Name="dyn_R_collar_Bace_1")
    Parent=(Type=Bone,Name="dyn_R_collar_Bace_1")
    bParentInitial=False
    bPropagateToChildren=True
    Weight=@n5.OutputParam
    Value=&RigVMFunction_MathTransformFromSRT(Location=(X=0,Y=0,Z=0), Rotation=(X=0,Y=30,Z=0), Scale=(X=1,Y=1,Z=1)).Transform
   ) {template="Set Relative Transform::Execute(in Child,in Parent,in bParentInitial,in Value,in Weight,in bPropagateToChildren)"}
   RigUnit_SequenceAggregate_2 = &RigVMFunction_Sequence(ExecuteContext=@RigUnit_SequenceAggregate.B)
   n6 = &RigUnit_SphericalPoseReader(
    ActiveRegionScaleFactors=(PositiveWidth=1,NegativeWidth=0,PositiveHeight=1,NegativeHeight=1)
    ActiveRegionSize=0.3
    Debug=(bDrawDebug=false,bDraw2D=false,bDrawLocalAxes=false,DebugScale=30,DebugSegments=30,DebugThickness=0.2)
    DriverAxis=(X=-1,Y=0,Z=0)
    DriverItem=(Type=Bone,Name="thigh_l")
    FalloffRegionScaleFactors=(PositiveWidth=1,NegativeWidth=1,PositiveHeight=1,NegativeHeight=1)
    FalloffSize=0.5
    OptionalParentItem=(Type=Bone,Name="pelvis")
    RotationOffset=(X=90,Y=0,Z=0)
    ExecutePin=@RigUnit_SequenceAggregate_2.A
   )
   > &RigUnit_SetRelativeTransformForItem(
    Child=(Type=Bone,Name="dyn_jacket_front_L_Base")
    Parent=(Type=Bone,Name="dyn_jacket_front_L_Base")
    bParentInitial=False
    bPropagateToChildren=True
    Weight=@n6.OutputParam
    Value=&RigVMFunction_MathTransformFromSRT(Location=(X=2,Y=-2,Z=0), Rotation=(X=50,Y=40,Z=-70), Scale=(X=1,Y=1,Z=1)).Transform
   ) {template="Set Relative Transform::Execute(in Child,in Parent,in bParentInitial,in Value,in Weight,in bPropagateToChildren)"}
   > &RigUnit_SetRelativeTransformForItem(
    Child=(Type=Bone,Name="dyn_jacket_side_L_Base")
    Parent=(Type=Bone,Name="dyn_jacket_side_L_Base")
    bParentInitial=False
    bPropagateToChildren=True
    Weight=@n6.OutputParam
    Value=&RigVMFunction_MathTransformFromSRT(Location=(X=-3,Y=0,Z=7), Rotation=(X=0,Y=-10,Z=0), Scale=(X=1,Y=1,Z=1)).Transform
   ) {template="Set Relative Transform::Execute(in Child,in Parent,in bParentInitial,in Value,in Weight,in bPropagateToChildren)"}
   > n7 = &RigUnit_SphericalPoseReader(
    ActiveRegionScaleFactors=(PositiveWidth=1,NegativeWidth=1,PositiveHeight=1,NegativeHeight=1)
    ActiveRegionSize=0.05
    Debug=(bDrawDebug=false,bDraw2D=false,bDrawLocalAxes=false,DebugScale=30,DebugSegments=30,DebugThickness=0.2)
    DriverAxis=(X=1,Y=0,Z=0)
    DriverItem=(Type=Bone,Name="spine_03")
    FalloffRegionScaleFactors=(PositiveWidth=1,NegativeWidth=1,PositiveHeight=1,NegativeHeight=1)
    FalloffSize=0.1
    OptionalParentItem=(Type=Bone,Name="pelvis")
    RotationOffset=(X=90,Y=0,Z=-60)
   )
   > &RigUnit_SetRelativeTransformForItem(
    Child=(Type=Bone,Name="dyn_jacket_back_M_Base")
    Parent=(Type=Bone,Name="dyn_jacket_back_M_Base")
    Value=(Rotation=(X=0,Y=0,Z=0,W=1),Translation=(X=3,Y=0,Z=0),Scale3D=(X=1,Y=1,Z=1))
    bParentInitial=False
    bPropagateToChildren=True
    Weight=@n7.OutputParam
   ) {template="Set Relative Transform::Execute(in Child,in Parent,in bParentInitial,in Value,in Weight,in bPropagateToChildren)"}
   > &RigUnit_SetRelativeTransformForItem(
    Child=(Type=Bone,Name="dyn_jacket_back_R_Base")
    Parent=(Type=Bone,Name="dyn_jacket_back_R_Base")
    Value=(Rotation=(X=0,Y=0,Z=0,W=1),Translation=(X=3,Y=0,Z=0),Scale3D=(X=1,Y=1,Z=1))
    bParentInitial=False
    bPropagateToChildren=True
    Weight=@n7.OutputParam
   ) {template="Set Relative Transform::Execute(in Child,in Parent,in bParentInitial,in Value,in Weight,in bPropagateToChildren)"}
   > &RigUnit_SetRelativeTransformForItem(
    Child=(Type=Bone,Name="dyn_jacket_back_L_Base")
    Parent=(Type=Bone,Name="dyn_jacket_back_L_Base")
    Value=(Rotation=(X=0,Y=0,Z=0,W=1),Translation=(X=3,Y=0,Z=0),Scale3D=(X=1,Y=1,Z=1))
    bParentInitial=False
    bPropagateToChildren=True
    Weight=@n7.OutputParam
   ) {template="Set Relative Transform::Execute(in Child,in Parent,in bParentInitial,in Value,in Weight,in bPropagateToChildren)"}
   > n8 = &RigUnit_SphericalPoseReader(
    ActiveRegionScaleFactors=(PositiveWidth=1.5,NegativeWidth=1,PositiveHeight=1,NegativeHeight=1)
    ActiveRegionSize=0.05
    Debug=(bDrawDebug=false,bDraw2D=false,bDrawLocalAxes=false,DebugScale=30,DebugSegments=30,DebugThickness=0.2)
    DriverAxis=(X=1,Y=0,Z=0)
    DriverItem=(Type=Bone,Name="spine_04")
    FalloffRegionScaleFactors=(PositiveWidth=3,NegativeWidth=1,PositiveHeight=1,NegativeHeight=1)
    FalloffSize=0.1
    OptionalParentItem=(Type=Bone,Name="spine_05")
    RotationOffset=(X=90,Y=0,Z=-65)
   )
   > &RigUnit_ModifyTransforms(
    ItemToModify="((Item=(Type=Bone,Name=\"dyn_jacket_back_M_01\"),Transform=(Rotation=(X=0,Y=0,Z=0,W=1),Translation=(X=0,Y=-3,Z=0),Scale3D=(X=1,Y=1,Z=1))),(Item=(Type=Bone,Name=\"dyn_jacket_back_R_Base\"),Transform=(Rotation=(X=0,Y=0,Z=0,W=1),Translation=(X=0,Y=-5,Z=0),Scale3D ...(truncated)"
    Weight=@n8.OutputParam
   )
   n9 = &RigUnit_SphericalPoseReader(
    ActiveRegionScaleFactors=(PositiveWidth=1,NegativeWidth=0,PositiveHeight=1,NegativeHeight=1)
    ActiveRegionSize=0.3
    Debug=(bDrawDebug=false,bDraw2D=false,bDrawLocalAxes=false,DebugScale=30,DebugSegments=30,DebugThickness=0.2)
    DriverAxis=(X=1,Y=0,Z=0)
    DriverItem=(Type=Bone,Name="thigh_r")
    FalloffRegionScaleFactors=(PositiveWidth=1,NegativeWidth=1,PositiveHeight=1,NegativeHeight=1)
    FalloffSize=0.5
    OptionalParentItem=(Type=Bone,Name="pelvis")
    RotationOffset=(X=-90,Y=180,Z=0)
    ExecutePin=@RigUnit_SequenceAggregate_2.B
   )
   > RigUnit_SetRelativeTransformForItem = &RigUnit_SetRelativeTransformForItem(
    Child=(Type=Bone,Name="dyn_jacket_front_R_base")
    Parent=(Type=Bone,Name="dyn_jacket_front_R_base")
    bParentInitial=False
    bPropagateToChildren=True
    Weight=@n9.OutputParam
    Value=&RigVMFunction_MathTransformFromSRT(Location=(X=0,Y=0,Z=0), Rotation=(X=-60,Y=30,Z=-70), Scale=(X=1,Y=1,Z=1)).Transform
   ) {template="Set Relative Transform::Execute(in Child,in Parent,in bParentInitial,in Value,in Weight,in bPropagateToChildren)"}
   > n10 = &RigUnit_SphericalPoseReader(
    ActiveRegionScaleFactors=(PositiveWidth=1,NegativeWidth=1,PositiveHeight=1,NegativeHeight=1)
    ActiveRegionSize=0.1
    Debug=(bDrawDebug=false,bDraw2D=false,bDrawLocalAxes=false,DebugScale=30,DebugSegments=30,DebugThickness=0.2)
    DriverAxis=(X=0,Y=0,Z=-1)
    DriverItem=(Type=Bone,Name="spine_05")
    FalloffRegionScaleFactors=(PositiveWidth=1,NegativeWidth=1,PositiveHeight=1,NegativeHeight=1)
    FalloffSize=0.15
    OptionalParentItem=(Type=Bone,Name="pelvis")
    RotationOffset=(X=140,Y=0,Z=0)
   )
   > &RigUnit_SetRelativeTransformForItem(
    Child=(Type=Bone,Name="dyn_jacket_front_L_Base")
    Parent=(Type=Bone,Name="dyn_jacket_front_L_Base")
    bParentInitial=False
    bPropagateToChildren=True
    Weight=@n10.OutputParam
    Value=&RigVMFunction_MathTransformFromSRT(Location=(X=0,Y=0,Z=0), Rotation=(X=0,Y=20,Z=0), Scale=(X=1,Y=1,Z=1)).Transform
   ) {template="Set Relative Transform::Execute(in Child,in Parent,in bParentInitial,in Value,in Weight,in bPropagateToChildren)"}
   > n11 = &RigUnit_SphericalPoseReader(
    ActiveRegionScaleFactors=(PositiveWidth=1,NegativeWidth=1,PositiveHeight=1,NegativeHeight=1)
    ActiveRegionSize=0.05
    Debug=(bDrawDebug=false,bDraw2D=false,bDrawLocalAxes=false,DebugScale=30,DebugSegments=30,DebugThickness=0.2)
    DriverAxis=(X=0,Y=0,Z=-1)
    DriverItem=(Type=Bone,Name="spine_05")
    FalloffRegionScaleFactors=(PositiveWidth=1,NegativeWidth=1,PositiveHeight=1,NegativeHeight=1)
    FalloffSize=0.15
    OptionalParentItem=(Type=Bone,Name="pelvis")
    RotationOffset=(X=150,Y=-35,Z=0)
   )
   &RigUnit_ModifyTransforms(
    ExecutePin=()
    ItemToModify=((Item=(Type=Bone,Name="dyn_jacket_side_R_Base"),Transform=(Rotation=(X=0,Y=0,Z=0,W=1),Translation=(X=0,Y=0,Z=0),Scale3D=(X=1,Y=1,Z=1))))
    Weight=@n9.OutputParam
    ItemToModify.0.Transform=&RigVMFunction_MathTransformFromSRT(Location=(X=0,Y=0,Z=0), Rotation=(X=0,Y=10,Z=0), Scale=(X=1,Y=1,Z=1)).Transform
   )
   &RigUnit_ModifyTransforms(
    ExecutePin=()
    ItemToModify=((Item=(Type=Bone,Name="dyn_jacket_side_L_Base"),Transform=(Rotation=(X=0,Y=0,Z=0,W=1),Translation=(X=0,Y=0,Z=0),Scale3D=(X=1,Y=1,Z=1))))
    Weight=@n10.OutputParam
    ItemToModify.0.Transform=&RigVMFunction_MathTransformFromSRT(Location=(X=0,Y=0,Z=-10), Rotation=(X=0,Y=0,Z=0), Scale=(X=1,Y=1,Z=1)).Transform
   )
   &RigUnit_ModifyTransforms(
    ExecutePin=()
    ItemToModify=((Item=(Type=Bone,Name="dyn_jacket_side_L_Base"),Transform=(Rotation=(X=0,Y=0,Z=0,W=1),Translation=(X=0,Y=0,Z=0),Scale3D=(X=1,Y=1,Z=1))))
    Weight=@n11.OutputParam
    ItemToModify.0.Transform=&RigVMFunction_MathTransformFromSRT(Location=(X=4,Y=0,Z=-10), Rotation=(X=30,Y=0,Z=0), Scale=(X=1,Y=1,Z=1)).Transform
   )
   RigVMRerouteNode(Value=@n5.OutputParam)
  ]}
]
hierarchy [
 root {loc=<0 0 0>, rot=<0 0 0 1>}
 pelvis {parent=root, loc=<0 -0.5301 98.0571>, rot=<0.0127 0.707 -0.0127 -0.707>}
 spine_01 {parent=pelvis, loc=<6.1208 -0.9328 0>, rot=<0 0 0.034 -0.9994>}
 spine_02 {parent=spine_01, loc=<6.5996 -0.2095 0>, rot=<0 0 0.0238 -0.9997>}
 spine_03 {parent=spine_02, loc=<6.4914 -0.1237 0>, rot=<0 0 0.0149 -0.9999>}
 spine_04 {parent=spine_03, loc=<6.9598 0.0794 0>, rot=<0 0 -0.0132 -0.9999>}
 spine_05 {parent=spine_04, loc=<6.9987 0 0>, rot=<0 0 -0.0015 -1>}
 attach_cape {parent=spine_05, loc=<9.3216 -8.4501 0>, rot=<0.0616 -0.7044 -0.0616 -0.7044>}
 neck_01 {parent=spine_05, loc=<11.3138 0.1529 0>, rot=<0 0 -0.2003 -0.9797>}
 neck_02 {parent=neck_01, loc=<4.5627 0 0>, rot=<0 0 0.0504 -0.9987>}
 head {parent=neck_02, loc=<5.3914 0 0>, rot=<0 0 0.115 -0.9934>}
 attach_backpack {parent=spine_05, loc=<0 -8.35 0>, rot=<-0.0093 -0.707 0.0093 -0.707>}
 clavicle_l {parent=spine_05, loc=<5.4831 -0.1053 -3.2946>, rot=<0.0465 -0.7289 -0.009 -0.6829>}
 upperarm_l {parent=clavicle_l, loc=<9.8658 -1.2136 -0.4105>, rot=<0.0164 -0.3245 -0.0269 -0.9454>}
 lowerarm_l {parent=upperarm_l, loc=<26.465 -0.8345 0.259>, rot=<0.0036 0.0161 -0.2298 -0.9731>}
 hand_l {parent=lowerarm_l, loc=<22.7355 -0.3203 -0.6105>, rot=<-0.6784 -0.0223 0.0735 -0.7307>}
 thumb_01_l {parent=hand_l, loc=<1.3118 -0.3191 -2.459>, rot=<0.6546 -0.4042 0.0664 -0.6354>}
 thumb_02_l {parent=thumb_01_l, loc=<5.9336 -0.0448 0.0401>, rot=<-0.0037 0.0213 0.298 -0.9543>}
 thumb_03_l {parent=thumb_02_l, loc=<2.2223 0.3682 -0.0907>, rot=<0.0118 -0.013 0.0066 -0.9998>}
 index_metacarpal_l {parent=hand_l, loc=<1.6535 0.0183 -2.96>, rot=<0.0158 -0.0124 -0.0173 -0.9996>}
 index_01_l {parent=index_metacarpal_l, loc=<8.3008 0.0357 -0.0524>, rot=<-0.0044 0.0485 0.0924 -0.9945>}
 index_02_l {parent=index_01_l, loc=<3.3587 -0.3004 0.0452>, rot=<-0.002 0.0027 0.3225 -0.9466>}
 index_03_l {parent=index_02_l, loc=<2.5578 -0.123 -0.0252>, rot=<-0.002 0.0067 0.2558 -0.9667>}
 middle_metacarpal_l {parent=hand_l, loc=<2.1856 -0.1725 -0.5984>, rot=<0.006 0.0168 -0.0329 -0.9993>}
 middle_01_l {parent=middle_metacarpal_l, loc=<7.3275 -0.2447 0.0129>, rot=<-0.0208 0.0042 0.0858 -0.9961>}
 middle_02_l {parent=middle_01_l, loc=<4.0123 0.1401 0.1175>, rot=<-0.0218 0.0258 0.3512 -0.9357>}
 middle_03_l {parent=middle_02_l, loc=<2.6077 0.0534 0.1373>, rot=<0.0082 0.0188 0.1708 -0.9851>}
 ring_metacarpal_l {parent=hand_l, loc=<2.2074 -0.3625 1.3638>, rot=<-0.0284 0.0722 -0.0458 -0.9959>}
 ring_01_l {parent=ring_metacarpal_l, loc=<7.0295 -0.6197 -0.1446>, rot=<-0.0489 -0.0252 0.1892 -0.9804>}
 ring_02_l {parent=ring_01_l, loc=<3.6174 0.4647 -0.1027>, rot=<0 0 0.2806 -0.9598>}
 ring_03_l {parent=ring_02_l, loc=<2.4152 0.114 0.0199>, rot=<0 0 0.1189 -0.9929>}
 pinky_metacarpal_l {parent=hand_l, loc=<1.7616 -0.9753 3.3814>, rot=<-0.0397 0.0814 -0.069 -0.9935>}
 pinky_01_l {parent=pinky_metacarpal_l, loc=<7.1709 -0.4831 -0.1094>, rot=<-0.0654 -0.0035 0.2242 -0.9723>}
 pinky_02_l {parent=pinky_01_l, loc=<2.8274 0.0893 -0.1043>, rot=<0.0037 0.0011 0.2854 -0.9584>}
 pinky_03_l {parent=pinky_02_l, loc=<2.2046 0.0307 -0.0078>, rot=<-0.0002 0.0098 0.1445 -0.9895>}
 lowerarm_twist_02_l {parent=lowerarm_l, loc=<10.6707 0 0>, rot=<0 0 0 1>}
 lowerarm_twist_01_l {parent=lowerarm_l, loc=<18.316 0 0>, rot=<0 0 0 1>}
 deform_elbow_in_base_l {parent=lowerarm_l, loc=<0.1834 3.6062 -0.2572>, rot=<0 0 0 -1>}
 deform_elbow_in_l {parent=deform_elbow_in_base_l, loc=<0 0 0>, rot=<0 0 0 1>}
 deform_elbow_out_base_l {parent=lowerarm_l, loc=<0 -4 0>, rot=<0 0 0 1>}
 deform_elbow_out_l {parent=deform_elbow_out_base_l, loc=<0 0 0>, rot=<0 0 0 1>}
 upperarm_twist_01_l {parent=upperarm_l, loc=<7.2135 -0.1134 0>, rot=<0 0 0 1>}
 deform_upperarm_bk_base_l {parent=upperarm_twist_01_l, loc=<-6.4053 -4.3746 0.314>, rot=<0 0 0 -1>}
 deform_upperarm_bk_l {parent=deform_upperarm_bk_base_l, loc=<0 0 0>, rot=<0 0 0 1>}
 deform_upperarm_fr_base_l {parent=upperarm_twist_01_l, loc=<-5.6312 5.092 0.3354>, rot=<0 0 0 -1>}
 deform_upperarm_fr_l {parent=deform_upperarm_fr_base_l, loc=<0 0 0>, rot=<0 0 0 1>}
 deform_upperarm_up_base_l {parent=upperarm_twist_01_l, loc=<-8.891 0.4825 3.9517>, rot=<0 0 0 -1>}
 deform_upperarm_up_l {parent=deform_upperarm_up_base_l, loc=<0 0 0>, rot=<0 0 0 1>}
 upperarm_twist_02_l {parent=upperarm_l, loc=<19.7998 -0.1134 0>, rot=<0 0 0 1>}
 clavicle_r {parent=spine_05, loc=<5.4831 -0.1053 3.2946>, rot=<0.7289 0.0465 -0.6829 0.009>}
 upperarm_r {parent=clavicle_r, loc=<-9.8658 1.2136 0.4105>, rot=<0.0164 -0.3245 -0.0269 -0.9454>}
 lowerarm_r {parent=upperarm_r, loc=<-26.465 0.8345 -0.259>, rot=<0.0036 0.0161 -0.2298 -0.9731>}
 hand_r {parent=lowerarm_r, loc=<-22.7355 0.3203 0.6106>, rot=<-0.6784 -0.0223 0.0735 -0.7307>}
 thumb_01_r {parent=hand_r, loc=<-1.3118 0.3191 2.459>, rot=<0.6546 -0.4042 0.0664 -0.6354>}
 thumb_02_r {parent=thumb_01_r, loc=<-5.9336 0.0448 -0.0401>, rot=<-0.0037 0.0213 0.298 -0.9543>}
 thumb_03_r {parent=thumb_02_r, loc=<-2.2223 -0.3682 0.0907>, rot=<0.0118 -0.013 0.0066 -0.9998>}
 index_metacarpal_r {parent=hand_r, loc=<-1.6535 -0.0183 2.96>, rot=<0.0158 -0.0124 -0.0173 -0.9996>}
 index_01_r {parent=index_metacarpal_r, loc=<-8.3008 -0.0357 0.0524>, rot=<-0.0044 0.0485 0.0924 -0.9945>}
 index_02_r {parent=index_01_r, loc=<-3.3587 0.3004 -0.0452>, rot=<-0.002 0.0027 0.3225 -0.9466>}
 index_03_r {parent=index_02_r, loc=<-2.5578 0.123 0.0252>, rot=<-0.002 0.0067 0.2558 -0.9667>}
 middle_metacarpal_r {parent=hand_r, loc=<-2.1856 0.1725 0.5984>, rot=<0.006 0.0168 -0.0329 -0.9993>}
 middle_01_r {parent=middle_metacarpal_r, loc=<-7.3275 0.2447 -0.0129>, rot=<-0.0208 0.0042 0.0858 -0.9961>}
 middle_02_r {parent=middle_01_r, loc=<-4.0123 -0.1401 -0.1175>, rot=<-0.0218 0.0258 0.3512 -0.9357>}
 middle_03_r {parent=middle_02_r, loc=<-2.6077 -0.0534 -0.1373>, rot=<0.0082 0.0188 0.1708 -0.9851>}
 ring_metacarpal_r {parent=hand_r, loc=<-2.2074 0.3625 -1.3638>, rot=<-0.0284 0.0722 -0.0458 -0.9959>}
 ring_01_r {parent=ring_metacarpal_r, loc=<-7.0295 0.6197 0.1446>, rot=<-0.0489 -0.0252 0.1892 -0.9804>}
 ring_02_r {parent=ring_01_r, loc=<-3.6174 -0.4647 0.1027>, rot=<0 0 0.2806 -0.9598>}
 ring_03_r {parent=ring_02_r, loc=<-2.4152 -0.114 -0.0199>, rot=<0 0 0.1189 -0.9929>}
 pinky_metacarpal_r {parent=hand_r, loc=<-1.7616 0.9753 -3.3814>, rot=<-0.0397 0.0814 -0.069 -0.9935>}
 pinky_01_r {parent=pinky_metacarpal_r, loc=<-7.1709 0.4831 0.1094>, rot=<-0.0654 -0.0035 0.2242 -0.9723>}
 pinky_02_r {parent=pinky_01_r, loc=<-2.8274 -0.0893 0.1043>, rot=<0.0037 0.0011 0.2854 -0.9584>}
 pinky_03_r {parent=pinky_02_r, loc=<-2.2046 -0.0307 0.0078>, rot=<-0.0002 0.0098 0.1445 -0.9895>}
 lowerarm_twist_02_r {parent=lowerarm_r, loc=<-10.6707 0 0>, rot=<0 0 0 -1>}
 lowerarm_twist_01_r {parent=lowerarm_r, loc=<-18.316 0 0>, rot=<-0.1754 0.0115 0.0015 -0.9844>}
 deform_elbow_in_base_r {parent=lowerarm_r, loc=<-0.1834 -3.6062 0.2572>, rot=<0.0722 0.3332 0.2199 -0.914>}
 deform_elbow_in_r {parent=deform_elbow_in_base_r, loc=<0 0 0>, rot=<0 0 0 1>}
 deform_elbow_out_base_r {parent=lowerarm_r, loc=<0 4 -0.0001>, rot=<0 0 0 1>}
 deform_elbow_out_r {parent=deform_elbow_out_base_r, loc=<0 0 0>, rot=<0 0 0 1>}
 upperarm_twist_01_r {parent=upperarm_r, loc=<-7.2135 0.1134 0>, rot=<0 0 0 -1>}
 deform_upperarm_bk_base_r {parent=upperarm_twist_01_r, loc=<6.4053 4.3746 -0.314>, rot=<0 0 0 -1>}
 deform_upperarm_bk_r {parent=deform_upperarm_bk_base_r, loc=<0 0 0>, rot=<0 0 0 1>}
 deform_upperarm_fr_base_r {parent=upperarm_twist_01_r, loc=<5.6312 -5.092 -0.3354>, rot=<0 0 0 -1>}
 deform_upperarm_fr_r {parent=deform_upperarm_fr_base_r, loc=<0 0 0>, rot=<0 0 0 1>}
 deform_upperarm_up_base_r {parent=upperarm_twist_01_r, loc=<8.891 -0.4825 -3.9517>, rot=<0 0 0 -1>}
 deform_upperarm_up_r {parent=deform_upperarm_up_base_r, loc=<0 0 0>, rot=<0 0 0 1>}
 upperarm_twist_02_r {parent=upperarm_r, loc=<-19.7998 0.1134 0>, rot=<0 0 0 -1>}
 Joint_control {parent=spine_05, loc=<1.6491 -1.5142 4.2593>, rot=<0.0284 -0.7065 -0.0284 -0.7065>}
 dyn_R_collar_Bace_1 {parent=spine_05, loc=<6.5155 5.5028 6.6813>, rot=<-0.4108 0.5092 0.7537 0.0625>}
 dyn_R_collar_3_1 {parent=dyn_R_collar_Bace_1, loc=<0 0 0>, rot=<0 0 0 1>}
 dyn_R_collar_3_2 {parent=dyn_R_collar_3_1, loc=<-3.6815 -0.1279 -0.0358>, rot=<0.0002 -0.0262 -0.0087 -0.9996>}
 dyn_R_collar_3_end {parent=dyn_R_collar_3_2, loc=<-4.5515 1.1655 0.1856>, rot=<0 0 0.0436 -0.999>}
 dyn_L_collar_Bace_1 {parent=spine_05, loc=<6.9217 5.3036 -7.0416>, rot=<-0.5092 -0.4108 0.0625 -0.7537>}
 dyn_L_collar_3_1 {parent=dyn_L_collar_Bace_1, loc=<0 0 0>, rot=<0 0 0 1>}
 dyn_L_collar_3_2 {parent=dyn_L_collar_3_1, loc=<3.8051 -0.1512 -0.1856>, rot=<0.0002 -0.0262 -0.0087 -0.9996>}
 dyn_L_collar_3_end {parent=dyn_L_collar_3_2, loc=<3.8834 -1.1358 -0.1818>, rot=<0 0 0.0436 -0.999>}
 dyn_L_collar_Bace_2 {parent=spine_05, loc=<1.9261 8.2786 -7.2133>, rot=<-0.5129 -0.4855 0.0102 -0.7079>}
 dyn_L_collar_2_1 {parent=dyn_L_collar_Bace_2, loc=<0 0 0>, rot=<0 0 0 1>}
 dyn_L_collar_2_end {parent=dyn_L_collar_2_1, loc=<2.9239 0.5984 -0.2376>, rot=<0.0002 -0.0262 -0.0087 -0.9996>}
 dyn_R_collar_Bace_2 {parent=spine_05, loc=<1.9258 8.2786 7.2133>, rot=<-0.4855 0.5129 0.7079 0.0102>}
 dyn_R_collar_2_1 {parent=dyn_R_collar_Bace_2, loc=<0 0 0>, rot=<0 0 0 1>}
 dyn_R_collar_2_end {parent=dyn_R_collar_2_1, loc=<-3.1465 -0.3708 0.2378>, rot=<0.0002 -0.0262 -0.0087 -0.9996>}
 dyn_jacket_front_R_base {parent=spine_04, loc=<-7.1522 9.1364 9.0716>, rot=<0.0911 -0.4544 0.8839 -0.0622>}
 dyn_jacket_front_R_01 {parent=dyn_jacket_front_R_base, loc=<0 0 0>, rot=<0 0 0 1>}
 dyn_jacket_front_R_02 {parent=dyn_jacket_front_R_01, loc=<8.523 -0.3542 -0.0481>, rot=<-0.0866 0.0099 0.1128 -0.9898>}
 dyn_jacket_front_R_03 {parent=dyn_jacket_front_R_02, loc=<6.3706 0.438 -0.193>, rot=<0 0 0 1>}
 dyn_jacket_front_L_Base {parent=spine_04, loc=<-7.1512 9.2451 -9.2067>, rot=<-0.0911 0.4544 0.8839 -0.0622>}
 dyn_jacket_front_L_01 {parent=dyn_jacket_front_L_Base, loc=<0 0 0>, rot=<0 0 0 1>}
 dyn_jacket_front_L_02 {parent=dyn_jacket_front_L_01, loc=<8.3914 -0.1624 0.1042>, rot=<0.0866 -0.0099 0.1128 -0.9898>}
 dyn_jacket_front_L_03 {parent=dyn_jacket_front_L_02, loc=<6.6126 0.432 -0.0288>, rot=<0 0 0 1>}
 dyn_jacket_side_L_Base {parent=spine_04, loc=<-4.2684 0.7156 -11.5033>, rot=<0.0034 -0.8287 -0.5556 0.0669>}
 dyn_jacket_side_L_01 {parent=dyn_jacket_side_L_Base, loc=<0 0 0>, rot=<0.0009 0.0188 0.0502 -0.9986>}
 dyn_jacket_side_L_02 {parent=dyn_jacket_side_L_01, loc=<7.652 0 0>, rot=<0 0 0.1736 -0.9848>}
 dyn_jacket_side_L_03 {parent=dyn_jacket_side_L_02, loc=<8.9732 0.3945 -0.5756>, rot=<0 0 0 1>}
 dyn_jacket_side_R_Base {parent=spine_04, loc=<-4.2204 0.1395 11.2915>, rot=<0.0034 -0.8287 0.5556 -0.0669>}
 dyn_jacket_side_R_01 {parent=dyn_jacket_side_R_Base, loc=<0 0 0>, rot=<-0.0009 -0.0188 0.0502 -0.9986>}
 dyn_jacket_side_R_02 {parent=dyn_jacket_side_R_01, loc=<7.9541 0 0>, rot=<0 0 0.1736 -0.9848>}
 dyn_jacket_side_R_03 {parent=dyn_jacket_side_R_02, loc=<9.2094 -0.1112 0.1051>, rot=<0 0 0 1>}
 dyn_jacket_back_M_Base {parent=spine_04, loc=<-3.4571 -8.4816 0>, rot=<-0.0247 -0.9997 0 0>}
 dyn_jacket_back_M_01 {parent=dyn_jacket_back_M_Base, loc=<0 0 0>, rot=<0 0 -0.0037 -1>}
 dyn_jacket_back_M_02 {parent=dyn_jacket_back_M_01, loc=<7.5324 0 0>, rot=<0 0 0.2164 -0.9763>}
 dyn_jacket_back_M_03 {parent=dyn_jacket_back_M_02, loc=<7.9998 -0.0597 0>, rot=<0 0 0 1>}
 dyn_jacket_back_R_Base {parent=spine_04, loc=<-3.63 -5.9819 7.4247>, rot=<0.0052 -0.9499 0.3122 -0.0114>}
 dyn_jacket_back_R_01 {parent=dyn_jacket_back_R_Base, loc=<0 0 0>, rot=<-0.0002 0.0108 -0.0124 -0.9999>}
 dyn_jacket_back_R_02 {parent=dyn_jacket_back_R_01, loc=<7.865 0.0193 0.0199>, rot=<0 0 0.2164 -0.9763>}
 dyn_jacket_back_R_03 {parent=dyn_jacket_back_R_02, loc=<8.9954 -0.2236 -0.1826>, rot=<0 0 0 1>}
 dyn_jacket_back_L_Base {parent=spine_04, loc=<-3.5351 -6.5068 -7.596>, rot=<-0.0036 -0.9682 -0.2496 -0.0179>}
 dyn_jacket_back_L_01 {parent=dyn_jacket_back_L_Base, loc=<0 0 0>, rot=<0.0016 -0.0371 -0.0439 -0.9983>}
 dyn_jacket_back_L_02 {parent=dyn_jacket_back_L_01, loc=<7.9833 0 0>, rot=<0 0 0.2164 -0.9763>}
 dyn_jacket_back_L_03 {parent=dyn_jacket_back_L_02, loc=<8.9457 -0.7992 0.5793>, rot=<0 0 0 1>}
 thigh_l {parent=pelvis, loc=<-1.0855 -0.2109 -8.7053>, rot=<0.0009 0.0355 0.0155 -0.9993>}
 calf_l {parent=thigh_l, loc=<-42.5284 -0.0019 -0.7297>, rot=<0.0006 -0.0161 -0.0391 -0.9991>}
 foot_l {parent=calf_l, loc=<-46.7401 -2.5014 -0.592>, rot=<0.0427 -0.0238 0.0353 -0.9982>}
 ball_l {parent=foot_l, loc=<-3.5658 10.8066 -0.1544>, rot=<0.0001 -0.0001 -0.7063 -0.7079>}
 dyn_L_bow_1_1 {parent=foot_l, loc=<3.5887 6.6434 1.0785>, rot=<-0.3518 -0.7603 -0.4575 0.2981>}
 dyn_L_bow_2_1 {parent=foot_l, loc=<3.5992 6.6604 -0.1116>, rot=<-0.2649 -0.8871 0.1397 -0.3513>}
 calf_twist_02_l {parent=calf_l, loc=<-19.1546 -1.7943 -0.2638>, rot=<0.0237 0.0095 -0.0035 -0.9997>}
 calf_twist_01_l {parent=calf_l, loc=<-35.8483 -1.9197 -0.5794>, rot=<0.0237 0.0095 -0.0035 -0.9997>}
 dyn_L_tongue_1 {parent=calf_twist_01_l, loc=<-5.5511 -4.3665 -0.0865>, rot=<-0.0162 -0.0723 0.6974 -0.7129>}
 deform_knee_in_base_l {parent=calf_l, loc=<0.8031 -5.0231 0.2456>, rot=<0 0 0 -1>}
 deform_knee_in_l {parent=deform_knee_in_base_l, loc=<0 0 0>, rot=<0 0 0 1>}
 deform_knee_out_base_l {parent=calf_l, loc=<0 4 0>, rot=<0 0 0 1>}
 deform_knee_out_l {parent=deform_knee_out_base_l, loc=<0 0 0>, rot=<0 0 0 1>}
 thigh_twist_01_l {parent=thigh_l, loc=<-27.952 0 0>, rot=<0.0013 0 0 -1>}
 deform_groin_base_l {parent=thigh_l, loc=<0.4801 5.6776 -0.2486>, rot=<0 0 0 -1>}
 deform_groin_l {parent=deform_groin_base_l, loc=<0 0 0>, rot=<0 0 0 1>}
 thigh_r {parent=pelvis, loc=<-1.0855 -0.2109 8.7053>, rot=<0.0355 -0.0009 0.9993 0.0155>}
 calf_r {parent=thigh_r, loc=<42.5284 0.0019 0.7297>, rot=<0.0006 -0.0161 -0.0391 -0.9991>}
 foot_r {parent=calf_r, loc=<46.7401 2.5014 0.592>, rot=<0.0427 -0.0238 0.0353 -0.9982>}
 ball_r {parent=foot_r, loc=<3.5658 -10.8066 0.1544>, rot=<0.0001 -0.0001 -0.7063 -0.7079>}
 dyn_R_bow_1_1 {parent=foot_r, loc=<-4.1859 -6.9915 -0.0654>, rot=<-0.2367 -0.8736 0.2092 -0.3702>}
 dyn_R_bow_2_1 {parent=foot_r, loc=<-4.3954 -7.0745 -0.8287>, rot=<-0.3748 -0.7934 -0.3963 0.2701>}
 calf_twist_02_r {parent=calf_r, loc=<19.1546 1.7943 0.2638>, rot=<0.0237 0.0095 -0.0035 -0.9997>}
 calf_twist_01_r {parent=calf_r, loc=<35.8483 1.9197 0.5794>, rot=<0.0237 0.0095 -0.0035 -0.9997>}
 dyn_R_tongue_1 {parent=calf_twist_01_r, loc=<5.5512 4.3665 0.0865>, rot=<-0.0065 -0.0177 -0.7157 -0.6982>}
 deform_knee_in_base_r {parent=calf_r, loc=<-0.8031 5.0231 -0.2457>, rot=<0 0 0 -1>}
 deform_knee_in_r {parent=deform_knee_in_base_r, loc=<0 0 0>, rot=<0 0 0 1>}
 deform_knee_out_base_r {parent=calf_r, loc=<0 -4 0>, rot=<0 0 0 1>}
 deform_knee_out_r {parent=deform_knee_out_base_r, loc=<0 0 0>, rot=<0 0 0 1>}
 thigh_twist_01_r {parent=thigh_r, loc=<27.952 0 0>, rot=<0.0013 0 0 -1>}
 deform_groin_base_r {parent=thigh_r, loc=<-0.4801 -5.6776 0.2486>, rot=<0 0 0 -1>}
 deform_groin_r {parent=deform_groin_base_r, loc=<0 0 0>, rot=<0 0 0 1>}
 deform_glute_base_l {parent=pelvis, loc=<-7.3466 -9.0213 -9.1245>, rot=<0.0172 0.031 -0.4477 -0.8935>}
 deform_glute_l {parent=deform_glute_base_l, loc=<0 0 0>, rot=<0 0 0 1>}
 deform_glute_base_r {parent=pelvis, loc=<-7.3464 -9.0213 9.1245>, rot=<0.031 -0.0172 0.8935 -0.4477>}
 deform_glute_r {parent=deform_glute_base_r, loc=<0 0 0>, rot=<0 0 0 1>}
]
