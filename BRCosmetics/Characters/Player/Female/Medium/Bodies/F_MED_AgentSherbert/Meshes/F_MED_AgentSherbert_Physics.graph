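# Physics asset for the AgentSherbert medium female player body, set up for secondary motion of the left and right collar bones.
# Each side has two short constrained chains hanging from tiny "Bace" base capsules, with narrow angular limits and the left outer tip fully angle-locked to its parent.
# Collar bodies are mostly single capsules, though the right outer tip combines a sphere and three capsules.
# Unconstrained upper-arm capsules supply arm volumes alongside the collars.
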PhysicsAsset /BRCosmetics/Characters/Player/Female/Medium/Bodies/F_MED_AgentSherbert/Meshes/F_MED_AgentSherbert_Physics
bodies {
 dyn_L_collar_2_1 [{
   shapeType=Capsule
   length=3
   radius=0.7
   shapeName=dyn_L_collar_2_1_capsule
   center=<1.3074 1.1047 -1.0305>
   rotation=<rot 16.3056 -90 26.2402>
  }]
 dyn_L_collar_3_1 [{
   shapeType=Capsule
   length=2.5171
   radius=2.1
   shapeName=dyn_L_collar_3_1_capsule
   center=<0.7108 0.5968 8.8471>
   rotation=<rot -2.2418 -81.9003 -37.9787>
  }]
 dyn_L_collar_3_2 [{
   shapeType=Capsule
   length=4
   radius=1.6
   shapeName=dyn_L_collar_3_2_capsule
   center=<4.0606 -0.6773 0.8576>
   rotation=<rot 8.074 -90 -38.3931>
  }]
 dyn_L_collar_Bace_1 [{shapeType=Capsule, length=0.505, radius=0.505, shapeName=dyn_L_collar_Bace_1_capsule, rotation=<rot 0 -90 0>}]
 dyn_L_collar_Bace_2 [{shapeType=Capsule, length=0.505, radius=0.505, shapeName=dyn_L_collar_Bace_2_capsule, rotation=<rot 0 -90 0>}]
 dyn_R_collar_2_1 [{
   shapeType=Capsule
   length=3
   radius=0.7
   shapeName=dyn_R_collar_2_1_capsule
   center=<-1.2953 -0.5698 2.1866>
   rotation=<rot 8.8848 -81.6279 22.1792>
  }]
 dyn_R_collar_3_1 [{
   shapeType=Capsule
   length=5.507
   radius=1.5854
   shapeName=dyn_R_collar_3_1_capsule
   center=<-1.1123 -1.0235 0.422>
   rotation=<rot 10.7155 -90.2079 11.4036>
  }]
 dyn_R_collar_3_2 [
  {shapeType=Sphere, radius=3, shapeName=dyn_R_collar_3_2_sphere, center=<-7.586 2.3653 0.6595>}
  {
   shapeType=Capsule
   length=10.1128
   radius=1.6755
   shapeName=dyn_R_collar_3_2_capsule
   center=<-2.8903 0.1678 -1.6373>
   rotation=<rot 7.1159 -92.6646 -41.7335>
  }
  {
   shapeType=Capsule
   length=8
   radius=2
   shapeName=dyn_R_collar_3_2_capsule
   center=<-3.5625 -0.0961 -3.9931>
   rotation=<rot 7.1158 -92.6646 -92.6437>
  }
  {
   shapeType=Capsule
   length=10
   radius=1.5
   shapeName=dyn_R_collar_3_2_capsule
   center=<-1.6652 1.0344 2.9687>
   rotation=<rot 6.3574 -96.5457 78.6569>
  }
 ]
 dyn_R_collar_Bace_1 [{shapeType=Capsule, length=0.505, radius=0.505, shapeName=dyn_R_collar_Bace_1_capsule, rotation=<rot 0 -90 0>}]
 dyn_R_collar_Bace_2 [{shapeType=Capsule, length=0.505, radius=0.505, shapeName=dyn_R_collar_Bace_2_capsule, rotation=<rot 0 -90 0>}]
 upperarm_l [{
   shapeType=Capsule
   length=13.7579
   radius=5.2801
   shapeName=upperarm_l_capsule
   center=<6.2647 0.5427 -0.7413>
   rotation=<rot 16.3924 -90 -89.5159>
  }]
 upperarm_r [{
   shapeType=Capsule
   length=19.9423
   radius=5.2171
   shapeName=upperarm_r_capsule
   center=<-8.0981 0.4763 1.1216>
   rotation=<rot 1.8348 -90 -87.5984>
  }]
}
constraints [
 {bone1Name=dyn_L_collar_3_1, bone2Name=dyn_L_collar_Bace_1, swing1LimitDegrees=11, swing2LimitDegrees=0.4, twistLimitDegrees=0.4}
 {
  bone1Name=dyn_L_collar_3_2
  bone2Name=dyn_L_collar_3_1
  swing1LimitDegrees=5
  swing1Motion=Locked
  swing2LimitDegrees=2
  swing2Motion=Locked
  twistLimitDegrees=1
  twistMotion=Locked
 }
 {bone1Name=dyn_L_collar_2_1, bone2Name=dyn_L_collar_Bace_2, swing1LimitDegrees=11, swing2LimitDegrees=7, twistLimitDegrees=3}
 {bone1Name=dyn_R_collar_3_1, bone2Name=dyn_R_collar_Bace_1, swing1LimitDegrees=7, swing2LimitDegrees=5, twistLimitDegrees=0}
 {bone1Name=dyn_R_collar_3_2, bone2Name=dyn_R_collar_3_1, swing1LimitDegrees=5.88, swing2LimitDegrees=5, twistLimitDegrees=0}
 {bone1Name=dyn_R_collar_2_1, bone2Name=dyn_R_collar_Bace_2, swing1LimitDegrees=11, swing2LimitDegrees=7, twistLimitDegrees=3}
]
