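# Physics asset for the AgentSherbert female medium body that simulates the lower jacket around the torso and thighs.
# Seven front, side and back chains hang from bases locked to spine_04, each with three angle-limited segments, mostly capsules with boxes on front-left.
# Cross-constraints tie the outer back tips to the middle and each side tip to its front, linking neighboring panels.
# Pelvis, spine and thigh capsules provide body volumes.
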PhysicsAsset /BRCosmetics/Characters/Player/Female/Medium/Bodies/F_MED_AgentSherbert/Meshes/F_MED_AgentSherbert_Physics_Bottom
bodies {
 dyn_jacket_back_L_01 [{shapeType=Capsule, length=6, radius=4, shapeName=dyn_jacket_back_L_01_capsule, center=<0.7177 -3 0>}]
 dyn_jacket_back_L_02 [{shapeType=Capsule, length=4, radius=6, shapeName=dyn_jacket_back_L_02_capsule, center=<3.5816 -4 0>}]
 dyn_jacket_back_L_03 [{shapeType=Capsule, length=6.3, radius=5, shapeName=dyn_jacket_back_L_03_capsule, center=<3.3951 -1.5 0>}]
 dyn_jacket_back_L_Base [{shapeType=Capsule, length=0.505, radius=0.505, shapeName=dyn_jacket_back_L_Base_capsule}]
 dyn_jacket_back_M_01 [{shapeType=Capsule, length=6.5, radius=3.5, shapeName=dyn_jacket_back_M_01_capsule, center=<0.7177 -3.4286 0>}]
 dyn_jacket_back_M_02 [{shapeType=Capsule, length=4, radius=6, shapeName=dyn_jacket_back_M_02_capsule, center=<3.5816 -4 0>}]
 dyn_jacket_back_M_03 [{shapeType=Capsule, length=2.3993, radius=5, shapeName=dyn_jacket_back_M_03_capsule, center=<5.3889 -1 0>}]
 dyn_jacket_back_M_Base [{shapeType=Capsule, length=0.505, radius=0.505, shapeName=dyn_jacket_back_M_Base_capsule}]
 dyn_jacket_back_R_01 [{shapeType=Capsule, length=6, radius=4, shapeName=dyn_jacket_back_R_01_capsule, center=<0.7177 -3.5 0>}]
 dyn_jacket_back_R_02 [{shapeType=Capsule, length=4, radius=6, shapeName=dyn_jacket_back_R_02_capsule, center=<3.5816 -4 0>}]
 dyn_jacket_back_R_03 [{shapeType=Capsule, length=6.3, radius=5, shapeName=dyn_jacket_back_R_03_capsule, center=<3.3951 -1.5 0>}]
 dyn_jacket_back_R_Base [{shapeType=Capsule, length=0.505, radius=0.505, shapeName=dyn_jacket_back_R_Base_capsule}]
 dyn_jacket_front_L_01 [{shapeType=Capsule, length=7, radius=3, shapeName=dyn_jacket_front_L_01_capsule}]
 dyn_jacket_front_L_02 [{shapeType=Box, extentX=7.5, extentY=1, extentZ=10, shapeName=dyn_jacket_front_L_02_box, center=<3.0985 0 0.059>}]
 dyn_jacket_front_L_03 [
  {
   shapeType=Box
   extentX=10
   extentY=1
   extentZ=13.3842
   shapeName=dyn_jacket_front_L_03_box
   center=<4.5289 0.4722 -2.0252>
   rotation=<rot -7.765 2.0775 -4.8372>
  }
  {
   shapeType=Box
   extentX=9.3148
   extentY=0.5025
   extentZ=10.9396
   shapeName=dyn_jacket_front_L_03_box
   center=<2.0473 0.7945 -5.531>
   rotation=<rot -7.4816 4.0436 -28.5014>
  }
 ]
 dyn_jacket_front_L_Base [{shapeType=Capsule, length=0, radius=0, shapeName=dyn_jacket_front_L_Base_capsule}]
 dyn_jacket_front_R_01 [{shapeType=Capsule, length=7, radius=3, shapeName=dyn_jacket_front_R_01_capsule}]
 dyn_jacket_front_R_02 [{
   shapeType=Capsule
   length=5.0917
   radius=4.4205
   shapeName=dyn_jacket_front_R_02_capsule
   center=<2.2508 -3 1>
   rotation=<rot 9.6215 -89.8367 -3.7471>
  }]
 dyn_jacket_front_R_03 [{
   shapeType=Capsule
   length=9.2893
   radius=4
   shapeName=dyn_jacket_front_R_03_capsule
   center=<4.4082 -3 0>
   rotation=<rot 6.7662 -88.3723 -15.5568>
  }]
 dyn_jacket_front_R_base [{shapeType=Capsule, length=0, radius=0, shapeName=dyn_jacket_front_R_base_capsule}]
 dyn_jacket_side_L_01 [{shapeType=Capsule, length=5, radius=4.5, shapeName=dyn_jacket_side_L_01_capsule, center=<0.7177 -3 0>}]
 dyn_jacket_side_L_02 [{shapeType=Capsule, length=5, radius=4.5, shapeName=dyn_jacket_side_L_02_capsule, center=<3.5816 -3 0>}]
 dyn_jacket_side_L_03 [{shapeType=Capsule, length=4.5092, radius=5, shapeName=dyn_jacket_side_L_03_capsule, center=<3.3951 -4 0>}]
 dyn_jacket_side_L_Base [{shapeType=Capsule, length=0, radius=0, shapeName=dyn_jacket_side_L_Base_capsule}]
 dyn_jacket_side_R_01 [{shapeType=Capsule, length=5, radius=5, shapeName=dyn_jacket_side_R_01_capsule, center=<0.7177 -3 0>}]
 dyn_jacket_side_R_02 [{shapeType=Capsule, length=5, radius=5, shapeName=dyn_jacket_side_R_02_capsule, center=<3.5816 -3 0>}]
 dyn_jacket_side_R_03 [{shapeType=Capsule, length=4.5239, radius=5, shapeName=dyn_jacket_side_R_03_capsule, center=<3.3951 -4 0>}]
 dyn_jacket_side_R_Base [{shapeType=Capsule, length=0, radius=0, shapeName=dyn_jacket_side_R_Base_capsule}]
 pelvis [{shapeType=Capsule, length=11.1693, radius=9.9187, shapeName=pelvis_capsule, center=<0 -0.1478 0>}]
 spine_01 [{shapeType=Capsule, length=6.6092, radius=9.6715, shapeName=spine_01_capsule, center=<0.5828 1.3916 0>, rotation=<rot 0 -90 0>}]
 spine_02 [{shapeType=Capsule, length=7.4799, radius=8.3983, shapeName=spine_02_capsule, center=<0.3923 1.9299 0.075>, rotation=<rot -4.6295 -90 0.7267>}]
 spine_04 [{shapeType=Capsule, length=3.4261, radius=10.0903, shapeName=spine_04_capsule, center=<-4.9993 1.8625 0>}]
 thigh_l [{shapeType=Capsule, length=10.0465, radius=9.8894, center=<-11.2435 -1.0772 0.6111>, rotation=<rot -1.7262 -95.5808 -99.5914>}]
 thigh_r [{shapeType=Capsule, length=10.337, radius=9.5988, center=<11.222 0.7296 -0.7487>, rotation=<rot -1.7262 -95.5809 -98.29>}]
 thigh_twist_01_l [{shapeType=Capsule, length=12.2665, radius=7.5639, shapeName=thigh_twist_01_l_capsule, center=<0.0129 0 0.0846>, rotation=<rot 0 90 81.3578>}]
 thigh_twist_01_r [{
   shapeType=Capsule
   length=12.2665
   radius=7.5639
   shapeName=thigh_twist_01_l_capsule
   center=<-0.0129 0 -0.0846>
   rotation=<rot 0 -90 -81.3578>
  }]
}
constraints [
 {
  bone1Name=thigh_l
  bone2Name=pelvis
  swing1LimitDegrees=45
  swing1Motion=Locked
  swing2LimitDegrees=45
  swing2Motion=Locked
  twistLimitDegrees=45
  twistMotion=Locked
 }
 {
  bone1Name=thigh_r
  bone2Name=pelvis
  swing1LimitDegrees=45
  swing1Motion=Locked
  swing2LimitDegrees=45
  swing2Motion=Locked
  twistLimitDegrees=45
  twistMotion=Locked
 }
 {
  bone1Name=spine_01
  bone2Name=pelvis
  swing1LimitDegrees=45
  swing1Motion=Free
  swing2LimitDegrees=45
  swing2Motion=Free
  twistLimitDegrees=45
  twistMotion=Free
 }
 {bone1Name=dyn_jacket_front_L_01, bone2Name=dyn_jacket_front_L_Base, swing1LimitDegrees=35, swing2LimitDegrees=35, twistLimitDegrees=15}
 {bone1Name=dyn_jacket_front_L_02, bone2Name=dyn_jacket_front_L_01, swing1LimitDegrees=35, swing2LimitDegrees=35, twistLimitDegrees=15}
 {bone1Name=dyn_jacket_front_L_03, bone2Name=dyn_jacket_front_L_02, swing1LimitDegrees=35, swing2LimitDegrees=35, twistLimitDegrees=20}
 {bone1Name=dyn_jacket_front_R_01, bone2Name=dyn_jacket_front_R_base, swing1LimitDegrees=32, swing2LimitDegrees=32, twistLimitDegrees=15}
 {bone1Name=dyn_jacket_side_L_02, bone2Name=dyn_jacket_side_L_01, swing1LimitDegrees=30, swing2LimitDegrees=12, twistLimitDegrees=36.56}
 {bone1Name=dyn_jacket_side_R_02, bone2Name=dyn_jacket_side_R_01, swing1LimitDegrees=30, swing2LimitDegrees=12, twistLimitDegrees=2}
 {
  bone1Name=thigh_twist_01_l
  bone2Name=thigh_l
  swing1LimitDegrees=45
  swing1Motion=Free
  swing2LimitDegrees=45
  swing2Motion=Free
  twistLimitDegrees=45
  twistMotion=Free
 }
 {
  bone1Name=thigh_twist_01_r
  bone2Name=thigh_r
  swing1LimitDegrees=45
  swing1Motion=Free
  swing2LimitDegrees=45
  swing2Motion=Free
  twistLimitDegrees=45
  twistMotion=Free
 }
 {
  bone1Name=spine_02
  bone2Name=spine_01
  swing1LimitDegrees=45
  swing1Motion=Free
  swing2LimitDegrees=45
  swing2Motion=Free
  twistLimitDegrees=45
  twistMotion=Free
 }
 {bone1Name=dyn_jacket_front_R_02, bone2Name=dyn_jacket_front_R_01, swing1LimitDegrees=35, swing2LimitDegrees=35, twistLimitDegrees=15}
 {bone1Name=dyn_jacket_front_R_03, bone2Name=dyn_jacket_front_R_02, swing1LimitDegrees=35, swing2LimitDegrees=35, twistLimitDegrees=15}
 {bone1Name=dyn_jacket_back_L_02, bone2Name=dyn_jacket_back_L_01, swing1LimitDegrees=30, swing2LimitDegrees=12, twistLimitDegrees=2}
 {bone1Name=dyn_jacket_back_M_02, bone2Name=dyn_jacket_back_M_01, swing1LimitDegrees=30, swing2LimitDegrees=12, twistLimitDegrees=2}
 {bone1Name=dyn_jacket_back_R_01, bone2Name=dyn_jacket_back_R_Base, swing1LimitDegrees=30, swing2LimitDegrees=12, twistLimitDegrees=2}
 {bone1Name=dyn_jacket_back_R_02, bone2Name=dyn_jacket_back_R_01, swing1LimitDegrees=30, swing2LimitDegrees=12, twistLimitDegrees=2}
 {bone1Name=dyn_jacket_back_L_01, bone2Name=dyn_jacket_back_L_Base, swing1LimitDegrees=30, swing2LimitDegrees=12, twistLimitDegrees=2}
 {bone1Name=dyn_jacket_back_M_01, bone2Name=dyn_jacket_back_M_Base, swing1LimitDegrees=30, swing2LimitDegrees=12, twistLimitDegrees=2}
 {bone1Name=dyn_jacket_side_L_01, bone2Name=dyn_jacket_side_L_Base, swing1LimitDegrees=30, swing2LimitDegrees=25, twistLimitDegrees=2}
 {bone1Name=dyn_jacket_side_R_01, bone2Name=dyn_jacket_side_R_Base, swing1LimitDegrees=30, swing2LimitDegrees=25, twistLimitDegrees=2}
 {
  bone1Name=dyn_jacket_front_R_base
  bone2Name=spine_04
  swing1LimitDegrees=45
  swing1Motion=Locked
  swing2LimitDegrees=45
  swing2Motion=Locked
  twistLimitDegrees=45
  twistMotion=Locked
 }
 {
  bone1Name=dyn_jacket_front_L_Base
  bone2Name=spine_04
  swing1LimitDegrees=45
  swing1Motion=Locked
  swing2LimitDegrees=45
  swing2Motion=Locked
  twistLimitDegrees=45
  twistMotion=Locked
 }
 {
  bone1Name=dyn_jacket_side_L_Base
  bone2Name=spine_04
  swing1LimitDegrees=45
  swing1Motion=Locked
  swing2LimitDegrees=45
  swing2Motion=Locked
  twistLimitDegrees=45
  twistMotion=Locked
 }
 {
  bone1Name=dyn_jacket_back_M_Base
  bone2Name=spine_04
  swing1LimitDegrees=45
  swing1Motion=Locked
  swing2LimitDegrees=45
  swing2Motion=Locked
  twistLimitDegrees=45
  twistMotion=Locked
 }
 {
  bone1Name=dyn_jacket_side_R_Base
  bone2Name=spine_04
  swing1LimitDegrees=45
  swing1Motion=Locked
  swing2LimitDegrees=45
  swing2Motion=Locked
  twistLimitDegrees=45
  twistMotion=Locked
 }
 {
  bone1Name=dyn_jacket_back_R_Base
  bone2Name=spine_04
  swing1LimitDegrees=45
  swing1Motion=Locked
  swing2LimitDegrees=45
  swing2Motion=Locked
  twistLimitDegrees=45
  twistMotion=Locked
 }
 {
  bone1Name=dyn_jacket_back_L_Base
  bone2Name=spine_04
  swing1LimitDegrees=45
  swing1Motion=Locked
  swing2LimitDegrees=45
  swing2Motion=Locked
  twistLimitDegrees=45
  twistMotion=Locked
 }
 {
  bone1Name=spine_04
  bone2Name=spine_02
  swing1LimitDegrees=45
  swing1Motion=Locked
  swing2LimitDegrees=45
  swing2Motion=Locked
  twistLimitDegrees=45
  twistMotion=Locked
 }
 {bone1Name=dyn_jacket_back_L_03, bone2Name=dyn_jacket_back_L_02, swing1LimitDegrees=30, swing2LimitDegrees=12, twistLimitDegrees=2}
 {bone1Name=dyn_jacket_back_M_03, bone2Name=dyn_jacket_back_M_02, swing1LimitDegrees=30, swing2LimitDegrees=12, twistLimitDegrees=2}
 {bone1Name=dyn_jacket_back_R_03, bone2Name=dyn_jacket_back_R_02, swing1LimitDegrees=30, swing2LimitDegrees=12, twistLimitDegrees=2}
 {bone1Name=dyn_jacket_side_L_03, bone2Name=dyn_jacket_side_L_02, swing1LimitDegrees=50, swing2LimitDegrees=12, twistLimitDegrees=30.8}
 {bone1Name=dyn_jacket_side_R_03, bone2Name=dyn_jacket_side_R_02, swing1LimitDegrees=30, swing2LimitDegrees=12, twistLimitDegrees=2}
 {
  bone1Name=dyn_jacket_back_L_03
  bone2Name=dyn_jacket_back_M_03
  swing1LimitDegrees=45
  swing1Motion=Locked
  swing2LimitDegrees=45
  swing2Motion=Locked
  twistLimitDegrees=45
  twistMotion=Locked
 }
 {
  bone1Name=dyn_jacket_back_R_03
  bone2Name=dyn_jacket_back_M_03
  swing1LimitDegrees=45
  swing1Motion=Locked
  swing2LimitDegrees=45
  swing2Motion=Locked
  twistLimitDegrees=45
  twistMotion=Locked
 }
 {
  bone1Name=dyn_jacket_side_L_03
  bone2Name=dyn_jacket_front_L_03
  swing1LimitDegrees=45
  swing1Motion=Locked
  swing2LimitDegrees=45
  swing2Motion=Locked
  twistLimitDegrees=45
  twistMotion=Locked
 }
 {
  bone1Name=dyn_jacket_side_R_03
  bone2Name=dyn_jacket_front_R_03
  swing1LimitDegrees=45
  swing1Motion=Locked
  swing2LimitDegrees=45
  swing2Motion=Locked
  twistLimitDegrees=45
 }
]
