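# Animation Blueprint for the AgentSherbert female body cosmetic, layering procedural motion and corrective rigs onto the pose copied from its attached parent.
# A Control Rig feeds additive left-jacket offsets during slides, right-bow adjustments, two RigidBody simulations (spine_05, pelvis) and AnimDynamics for bows and tongues, then a deformation Control Rig.
# EventGraph updates RigidBody gravity through a deprecated state function and sets idle and gun_off_on flags, which the AnimGraph never reads.
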AnimBlueprint /BRCosmetics/Characters/Player/Female/Medium/Bodies/F_MED_AgentSherbert/Meshes/F_MED_AgentSherbert_AnimBP
target_skeleton=&/F_MED_AgentSherbert_Skeleton
graphs [
 AnimGraph {nodes [
   n1 = Control Rig(Source=Copy Pose From Mesh() {bUseAttachedParent=true}) {
    alphaBoolBlend {blendOption=Linear}
    controlRigAssetReference {blueprintRigClass=&/F_MED_AgentSherbert_CtrlRig.F_MED_AgentSherbert_CtrlRig_C}
   }
   n2 = "Transform (Modify) Bone - Bone: dyn_jacket_front_L_Base"!(
    ComponentPose=Local To Component(LocalPose=@n1)
    bAlphaBoolEnabled:bool=Get bIsKneeSliding()
    Translation="0.570,5.701,0.350"
    Rotation="-0.462,-3.179,-27.364"
    Scale="1,1,1"
   ) -> (Pose) {
    alphaInputType=Bool
    rotationMode=BMM_Additive
    translationMode=BMM_Additive
    alphaBoolBlend {blendOption=Linear}
    boneToModify {boneName=dyn_jacket_front_L_Base}
    rotation=<rot -0.462 -3.179 -27.364>
    translation=<0.57 5.701 0.35>
   }
   n3 = "Transform (Modify) Bone - Bone: dyn_jacket_front_L_Base"(
    ComponentPose=@n2
    bAlphaBoolEnabled=Get bIsBaseballSliding()
    Translation="0.000,5.176,4.761"
    Rotation="-5.114,-3.189,-27.100"
    Scale="1,1,1"
   ) {
    alphaInputType=Bool
    rotationMode=BMM_Additive
    translationMode=BMM_Additive
    alphaBoolBlend {blendOption=Linear}
    boneToModify {boneName=dyn_jacket_front_L_Base}
    rotation=<rot -5.114 -3.189 -27.1>
    translation {y=5.176, z=4.761}
   }
   n4 = "Transform (Modify) Bone - Bone: dyn_R_bow_1_1"(ComponentPose=@n3, Alpha=0.5, Translation="-0.673,-0.122,1.989", Rotation="36.696,0.001,0.006", Scale="1,1,1") {
    rotationMode=BMM_Additive
    rotationSpace=BCS_BoneSpace
    translationMode=BMM_Replace
    translationSpace=BCS_BoneSpace
    alphaBoolBlend {blendOption=Linear}
    boneToModify {boneName=dyn_R_bow_1_1}
    rotation=<rot 36.696 0.001 0.006>
    translation=<-0.673 -0.122 1.989>
   }
   n5 = "Transform (Modify) Bone - Bone: dyn_R_bow_2_1"(ComponentPose=@n4, Alpha=1, Translation="-0.639,0.363,-0.403", Rotation="4.966,-2.131,-5.629", Scale="1,1,1") {
    rotationMode=BMM_Additive
    translationMode=BMM_Replace
    translationSpace=BCS_BoneSpace
    alphaBoolBlend {blendOption=Linear}
    boneToModify {boneName=dyn_R_bow_2_1}
    rotation=<rot 4.966 -2.131 -5.629>
    translation=<-0.639 0.363 -0.403>
   }
   n6 = Get Gravity Override()
   n7 = RigidBody!(ComponentPose=@n5, Alpha, OverrideWorldGravity:Vector=@n6, ExternalForce) -> (Pose) {
    bClampLinearTranslationLimitToRefPose=true
    bOverrideWorldGravity=true
    overridePhysicsAsset=&/F_MED_AgentSherbert_Physics
    simulationSpace=BaseBoneSpace
    alphaBoolBlend {blendOption=Linear}
    baseBoneRef {boneName=spine_05}
    componentAppliedLinearAccClamp=<3000 3000 3000>
    componentLinearAccScale {z=0.1}
    componentLinearVelScale {z=0.1}
    overrideWorldGravity {z=-400}
    simSpaceSettings {
     maxAngularAcceleration=200
     maxAngularVelocity=3
     maxLinearAcceleration=200
     maxLinearVelocity=150
     velocityScaleZ=0
     worldAlpha=1
     externalLinearDragV=<0.1 0.1>
    }
   }
   n8 = RigidBody(ComponentPose=@n7, OverrideWorldGravity=@n6) {
    bClampLinearTranslationLimitToRefPose=true
    bOverrideWorldGravity=true
    overridePhysicsAsset=&/F_MED_AgentSherbert_Physics_Bottom
    simulationSpace=BaseBoneSpace
    alphaBoolBlend {blendOption=Linear}
    baseBoneRef {boneName=pelvis}
    componentAppliedLinearAccClamp=<2000 2000 2000>
    overrideWorldGravity {z=-400}
    simSpaceSettings {
     maxAngularAcceleration=250
     maxAngularVelocity=2
     maxLinearAcceleration=150
     maxLinearVelocity=100
     velocityScaleZ=0
     worldAlpha=1
     externalLinearDragV=<1 1 1>
    }
   }
   n9 = Anim Dynamics - Bone: dyn_L_bow_1_1(ComponentPose=@n8, Alpha=1, ExternalForce="0,0,0") {
    angularDampingOverride=0.9999
    angularSpringConstant=50
    bAngularSpring=true
    bOverrideAngularDamping=true
    bUseGravityOverride=true
    simulationSpace=RootRelative
    alphaBoolBlend {blendOption=Linear}
    boundBone {boneName=dyn_L_bow_1_1}
    componentAppliedLinearAccClamp=<1000 1000 1000>
    componentLinearAccScale=<1 1 1>
    componentLinearVelScale=<1 1 1>
    gravityOverride {z=-920}
    physicsBodyDefinitions [{
      boundBone {boneName=dyn_L_bow_1_1}
      boxExtents=<2 2 2>
      constraintSetup {angularLimitsMax=<0 30 20>, angularLimitsMin=<0 -30 -3>, angularTarget=<0 2 0>}
      localJointOffset=<0.1 0 0>
     }]
    relativeSpaceBone {boneName=calf_twist_02_r}
   }
   n10 = Anim Dynamics - Bone: dyn_L_bow_2_1(ComponentPose=@n9, Alpha=1, ExternalForce="0,0,0") {
    angularDampingOverride=0.9999
    angularSpringConstant=50
    bAngularSpring=true
    bOverrideAngularDamping=true
    bUseGravityOverride=true
    simulationSpace=RootRelative
    alphaBoolBlend {blendOption=Linear}
    boundBone {boneName=dyn_L_bow_2_1}
    componentAppliedLinearAccClamp=<1000 1000 1000>
    componentLinearAccScale=<1 1 1>
    componentLinearVelScale=<1 1 1>
    gravityOverride {z=-920}
    physicsBodyDefinitions [{
      boundBone {boneName=dyn_L_bow_2_1}
      boxExtents=<2 2 2>
      constraintSetup {angularLimitsMax=<30 10 0>, angularLimitsMin=<-30 -40 0>, angularTarget=<0 0 1>}
      localJointOffset=<0.1 0 0>
     }]
    relativeSpaceBone {boneName=calf_twist_02_r}
   }
   n11 = Anim Dynamics - Bone: dyn_L_tongue_1(ComponentPose=@n10, Alpha=1, ExternalForce="0,0,0") {
    angularDampingOverride=1
    angularSpringConstant=100
    bAngularSpring=true
    bOverrideAngularDamping=true
    bUseGravityOverride=true
    simulationSpace=RootRelative
    alphaBoolBlend {blendOption=Linear}
    boundBone {boneName=dyn_L_tongue_1}
    componentAppliedLinearAccClamp=<1000 1000 1000>
    componentLinearAccScale=<0.3 0.3 0.3>
    componentLinearVelScale=<0.3 0.3 0.3>
    gravityOverride {z=-980}
    physicsBodyDefinitions [{
      boundBone {boneName=dyn_L_tongue_1}
      boxExtents=<5 5 5>
      constraintSetup {angularLimitsMax=<0 0 30>, angularLimitsMin=<0 0 -30>, angularTarget=<1 0 3>, linearAxesMax=<1 0 0>}
      localJointOffset=<1 1 1>
     }]
    relativeSpaceBone {boneName=calf_twist_02_r}
   }
   n12 = Anim Dynamics - Bone: dyn_R_bow_1_1(ComponentPose=@n11, Alpha=1, ExternalForce="0,0,0") {
    angularDampingOverride=0.9999
    angularSpringConstant=50
    bAngularSpring=true
    bOverrideAngularDamping=true
    bUseGravityOverride=true
    simulationSpace=RootRelative
    alphaBoolBlend {blendOption=Linear}
    boundBone {boneName=dyn_R_bow_1_1}
    componentAppliedLinearAccClamp=<1000 1000 1000>
    componentLinearAccScale=<1 1 1>
    componentLinearVelScale=<1 1 1>
    gravityOverride {z=-920}
    physicsBodyDefinitions [{
      boundBone {boneName=dyn_R_bow_1_1}
      boxExtents=<2 2 2>
      constraintSetup {angularLimitsMax=<5 5 10>, angularLimitsMin=<-20 -20 -20>, angularTarget=<0 0 -0.5>}
      localJointOffset=<0.1 0 0>
     }]
    relativeSpaceBone {boneName=calf_twist_02_r}
   }
   n13 = Anim Dynamics - Bone: dyn_R_bow_2_1(ComponentPose=@n12, Alpha=1, ExternalForce="0,0,0") {
    angularDampingOverride=0.9999
    angularSpringConstant=50
    bAngularSpring=true
    bOverrideAngularDamping=true
    bUseGravityOverride=true
    simulationSpace=RootRelative
    alphaBoolBlend {blendOption=Linear}
    boundBone {boneName=dyn_R_bow_2_1}
    componentAppliedLinearAccClamp=<1000 1000 1000>
    componentLinearAccScale=<1 1 1>
    componentLinearVelScale=<1 1 1>
    gravityOverride {z=-920}
    physicsBodyDefinitions [{
      boundBone {boneName=dyn_R_bow_2_1}
      boxExtents=<2 2 2>
      constraintSetup {angularLimitsMax=<30 5 30>, angularLimitsMin=<-5 -20 -5>, angularTarget=<0 -1 0>}
      localJointOffset=<0.1 0 0>
     }]
    relativeSpaceBone {boneName=calf_twist_02_r}
   }
   n14 = Anim Dynamics - Bone: dyn_R_tongue_1(ComponentPose=@n13, Alpha=1, ExternalForce="0,0,0") {
    angularDampingOverride=1
    angularSpringConstant=100
    bAngularSpring=true
    bOverrideAngularDamping=true
    bUseGravityOverride=true
    simulationSpace=RootRelative
    alphaBoolBlend {blendOption=Linear}
    boundBone {boneName=dyn_R_tongue_1}
    componentAppliedLinearAccClamp=<1000 1000 1000>
    componentLinearAccScale=<0.3 0.3 0.3>
    componentLinearVelScale=<0.3 0.3 0.3>
    gravityOverride {z=-980}
    physicsBodyDefinitions [{
      boundBone {boneName=dyn_R_tongue_1}
      boxExtents=<5 5 5>
      constraintSetup {angularLimitsMax=<0 0 30>, angularLimitsMin=<0 0 -30>, angularTarget=<1 0 3>, linearAxesMax=<1 0 0>}
      localJointOffset=<1 1 1>
     }]
    relativeSpaceBone {boneName=calf_twist_02_r}
   }
   n15 = Control Rig(Source=Component To Local(ComponentPose=@n14)) {
    alphaBoolBlend {blendOption=Linear}
    controlRigAssetReference {blueprintRigClass=&/F_MED_AgentSherbert_Deforms_ControlRig.F_MED_AgentSherbert_Deforms_ControlRig_C}
   }
   Output Pose(Result=@n15) {layerGroup=None}
  ]}
 EventGraph {nodes [
   Event BlueprintUpdateAnimation()
   > n1 = Sequence()
   > n2 = Gravity Override State Function DEPRECATED(CharacterPartAnimInstance=Self-Reference(), GameTimeInSeconds=GetGameTimeInSeconds(), GravityOverrideParams=Get Gravity Override Params())
   > Set Gravity Override(Gravity Override=@n2)
   n3 = Sequence!(execute=@n1.then_1) -> (then_0, then_1, then_2:Exec)
   n4 = Get bIsMoving2D()
   > n5 = Branch(Condition=AND Boolean(A=Get bIsOnGround() -> (bIsOnGround:bool), B=NOT Boolean(A=@n4)))
   > Set idle(idle:bool=true) -> (Output_Get:bool)
   Set idle(execute=@n5.else, idle=false)
   n6 = Branch(execute=@n3.then_1, Condition=OR Boolean!(A=Get bIsCrouching(), B=Get bIsMoving2D(), C:bool=Get bIsSprinting()) -> (ReturnValue))
   > Set gun_off_on(gun_off_on:bool=true) -> (Output_Get:bool)
   Set gun_off_on(execute=@n6.else, gun_off_on=false)
   Event BlueprintInitializeAnimation()
   > Set Skel_Mesh(Skel_Mesh=GetOwnerSkeletalMeshComponent())
  ]}
]
variables [
 Gravity Override {type=Vector}
 Gravity Override Params {
  type=GravityOverrideParamsStruct
  default {
   skydive_FloatingDown {gravityOverride=<0 200 500>, windAmplitude=<0 100 200>, windFrequency=<0 35 35>, windOffset=<0 -400 -300>}
   skydive_GoingUp {gravityOverride=<0 200 500>, windAmplitude=<0 100 200>, windFrequency=<20 35 35>, windOffset=<0 -400 -300>}
  }
 }
 gun_off_on {type=boolean}
 idle {type=boolean}
 Skel_Mesh {type=SkeletalMeshComponent}
 thigh_rot_l {type=number}
 thigh_rot_r {type=number}
]
