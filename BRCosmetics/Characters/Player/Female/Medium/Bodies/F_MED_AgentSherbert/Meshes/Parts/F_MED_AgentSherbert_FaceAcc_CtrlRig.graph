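# Control Rig for Agent Sherbert's face accessory that adds pose-driven corrective offsets to hair-tail and dyn_head bones.
# It parent-constrains dyn_star to weighted neck, head and clavicle bones, then branches on "Is dew" and "is sky".
# Spherical pose readers on head, spine_05, upperarm_r and pelvis weight Modify Transforms offsets, some multiplied to require combined poses.
# "is sky" branches apply only zero offsets; several other chains, including a Sequence, are disconnected.
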ControlRigBlueprint /BRCosmetics/Characters/Player/Female/Medium/Bodies/F_MED_AgentSherbert/Meshes/Parts/F_MED_AgentSherbert_FaceAcc_CtrlRig
graphs [
 RigVMFunctionLibrary {links [], nodes []}
 RigVMModel {nodes [
   &RigUnit_ModifyTransforms(
    ExecutePin=()
    ItemToModify=((Item=(Type=Bone,Name="dyn_head"),Transform=(Rotation=(X=0,Y=0,Z=0,W=1),Translation=(X=0,Y=0,Z=0),Scale3D=(X=1,Y=1,Z=1))))
    Weight=1
    ItemToModify.0.Transform=/Script/RigVM.RigVMFunction_MathTransformFromSRT(Location=(X=1,Y=-1,Z=0), Rotation=(X=0,Y=0,Z=0), Scale=(X=1,Y=1,Z=1)) {is=struct RigVMUnitNode}.Transform
   )
   > &RigUnit_ModifyTransforms(
    ItemToModify=((Item=(Type=Bone,Name="dyn_hair_tail_1_l"),Transform=(Rotation=(X=0,Y=0,Z=0,W=1),Translation=(X=0,Y=0,Z=0),Scale3D=(X=1,Y=1,Z=1))))
    Weight=1
    ItemToModify.0.Transform=&RigVMFunction_MathTransformFromSRT(Location=(X=0,Y=0,Z=0), Rotation=(X=0,Y=0,Z=10), Scale=(X=1,Y=1,Z=1)).Transform
   )
   RigUnit_BeginExecution = &RigUnit_BeginExecution()
   n1 = &RigUnit_ParentConstraint(
    AdvancedSettings=(RotationOrderForFilter=XZY)
    Child=(Type=Bone,Name="dyn_star")
    Parents="((Item=(Type=Bone,Name=\"neck_01\"),Weight=0.3814),(Item=(Type=Bone,Name=\"neck_02\"),Weight=0.3656),(Item=(Type=Bone,Name=\"head\"),Weight=0.1061),(Item=(Type=Bone,Name=\"clavicle_l\"),Weight=0.0139),(Item=(Type=Bone,Name=\"clavicle_r\"),Weight=0.0029),(Item=(Type= ...(truncated)"
    Weight=1
    ExecutePin=@RigUnit_BeginExecution.ExecutePin
   )
   RigVMFunction_ControlFlowBranch = &RigVMFunction_ControlFlowBranch(ExecuteContext=@n1.ExecutePin, Condition=RigVMVariableNode() {variable=Is dew}.Value)
   n2 = RigVMVariableNode() {variable=is sky}
   n3 = &RigVMFunction_ControlFlowBranch(ExecuteContext=@RigVMFunction_ControlFlowBranch.False, Condition=@n2.Value)
   n4 = &RigUnit_SphericalPoseReader(
    ActiveRegionScaleFactors=(PositiveWidth=4,NegativeWidth=1,PositiveHeight=6,NegativeHeight=1)
    ActiveRegionSize=0.035
    Debug=(bDrawDebug=false,bDraw2D=false,DebugScale=25,DebugSegments=20,DebugThickness=0.25)
    DriverAxis=(X=1,Y=0,Z=0)
    DriverItem=(Type=Bone,Name="head")
    FalloffRegionScaleFactors=(PositiveWidth=1,NegativeWidth=1,PositiveHeight=1,NegativeHeight=0.3)
    FalloffSize=0.3
    OptionalParentItem=(Type=Bone)
    RotationOffset=(X=0,Y=90,Z=160)
    ExecutePin=@n3.False
   )
   > &RigUnit_ModifyTransforms(
    ItemToModify=((Item=(Type=Bone,Name="dyn_hair_tair_1_r"),Transform=(Rotation=(X=0,Y=0,Z=0,W=1),Translation=(X=0,Y=0,Z=0),Scale3D=(X=1,Y=1,Z=1))))
    ItemToModify.0.Transform=&RigVMFunction_MathTransformFromSRT(Location=(X=-2,Y=-3,Z=0), Rotation=(X=0,Y=0,Z=10), Scale=(X=1,Y=1,Z=1)).Transform
    Weight=@n4.OutputParam
   )
   > &RigUnit_ModifyTransforms(
    ItemToModify=((Item=(Type=Bone,Name="dyn_hair_tail_1_l"),Transform=(Rotation=(X=0,Y=0,Z=0,W=1),Translation=(X=0,Y=0,Z=0),Scale3D=(X=1,Y=1,Z=1))))
    ItemToModify.0.Transform=&RigVMFunction_MathTransformFromSRT(Location=(X=0,Y=2,Z=0), Rotation=(X=0,Y=0,Z=0), Scale=(X=1,Y=1,Z=1)).Transform
    Weight=@n4.OutputParam
   )
   > n5 = &RigUnit_SphericalPoseReader(
    ActiveRegionScaleFactors=(PositiveWidth=4,NegativeWidth=1,PositiveHeight=0.2,NegativeHeight=1)
    ActiveRegionSize=0.035
    Debug=(bDrawDebug=false,bDraw2D=false,DebugScale=25,DebugSegments=20,DebugThickness=0.25)
    DriverAxis=(X=0,Y=1,Z=0)
    DriverItem=(Type=Bone,Name="spine_05")
    FalloffRegionScaleFactors=(PositiveWidth=1,NegativeWidth=1,PositiveHeight=0.2,NegativeHeight=0.5)
    FalloffSize=0.3
    OptionalParentItem=(Type=Bone,Name="root")
    RotationOffset=(X=6,Y=-10,Z=90)
   )
   n6 = &RigVMFunction_MathFloatMul(A=@n5.OutputParam, B=@n4.OutputParam) {template="Multiply::Execute(in A,in B,out Result)"}
   > &RigUnit_ModifyTransforms(
    ItemToModify=((Item=(Type=Bone,Name="dyn_hair_tair_1_r"),Transform=(Rotation=(X=0,Y=0,Z=0,W=1),Translation=(X=0,Y=0,Z=0),Scale3D=(X=1,Y=1,Z=1))))
    ItemToModify.0.Transform=&RigVMFunction_MathTransformFromSRT(Location=(X=0,Y=2,Z=0), Rotation=(X=0,Y=15,Z=0), Scale=(X=1,Y=1,Z=1)).Transform
    Weight=@n6.Result
   )
   > &RigUnit_ModifyTransforms(
    ItemToModify=((Item=(Type=Bone,Name="dyn_hair_tail_1_l"),Transform=(Rotation=(X=0,Y=0,Z=0,W=1),Translation=(X=0,Y=0,Z=0),Scale3D=(X=1,Y=1,Z=1))))
    ItemToModify.0.Transform=&RigVMFunction_MathTransformFromSRT(Location=(X=0,Y=0,Z=0), Rotation=(X=0,Y=0,Z=0), Scale=(X=1,Y=1,Z=1)).Transform
    Weight=@n6.Result
   )
   > n7 = &RigUnit_SphericalPoseReader(
    ActiveRegionScaleFactors=(PositiveWidth=4,NegativeWidth=1,PositiveHeight=0.2,NegativeHeight=1)
    ActiveRegionSize=0.035
    Debug=(bDrawDebug=false,bDraw2D=false,DebugScale=25,DebugSegments=20,DebugThickness=0.25)
    DriverAxis=(X=0,Y=1,Z=0)
    DriverItem=(Type=Bone,Name="spine_05")
    FalloffRegionScaleFactors=(PositiveWidth=0.8,NegativeWidth=1,PositiveHeight=0.2,NegativeHeight=0.3)
    FalloffSize=0.3
    OptionalParentItem=(Type=Bone,Name="root")
    RotationOffset=(X=28,Y=0,Z=90)
   )
   > &RigUnit_ModifyTransforms(
    ItemToModify=((Item=(Type=Bone,Name="dyn_hair_tair_1_r"),Transform=(Rotation=(X=0,Y=0,Z=0,W=1),Translation=(X=0,Y=0,Z=0),Scale3D=(X=1,Y=1,Z=1))))
    ItemToModify.0.Transform=&RigVMFunction_MathTransformFromSRT(Location=(X=0,Y=0,Z=0), Rotation=(X=0,Y=15,Z=0), Scale=(X=1,Y=1,Z=1)).Transform
    Weight=@n7.OutputParam
   )
   > &RigUnit_ModifyTransforms(
    ItemToModify=((Item=(Type=Bone,Name="dyn_hair_tail_1_l"),Transform=(Rotation=(X=0,Y=0,Z=0,W=1),Translation=(X=0,Y=0,Z=0),Scale3D=(X=1,Y=1,Z=1))))
    ItemToModify.0.Transform=&RigVMFunction_MathTransformFromSRT(Location=(X=0,Y=0,Z=0), Rotation=(X=0,Y=0,Z=0), Scale=(X=1,Y=1,Z=1)).Transform
    Weight=&RigVMFunction_MathFloatMul(A=@n7.OutputParam, B=@n4.OutputParam) {template="Multiply::Execute(in A,in B,out Result)"}.Result
   )
   > n8 = &RigUnit_SphericalPoseReader(
    ActiveRegionScaleFactors=(PositiveWidth=1,NegativeWidth=1,PositiveHeight=1,NegativeHeight=1)
    ActiveRegionSize=0.04
    Debug=(bDrawDebug=false,bDraw2D=false,DebugScale=25,DebugSegments=20,DebugThickness=0.25)
    DriverAxis=(X=0,Y=1,Z=-1)
    DriverItem=(Type=Bone,Name="upperarm_r")
    FalloffRegionScaleFactors=(PositiveWidth=1,NegativeWidth=1,PositiveHeight=1,NegativeHeight=1)
    FalloffSize=0.13
    OptionalParentItem=(Type=Bone,Name="pelvis")
    RotationOffset=(X=50,Y=120,Z=17)
   )
   > n9 = &RigUnit_SphericalPoseReader(
    ActiveRegionScaleFactors=(PositiveWidth=1,NegativeWidth=2,PositiveHeight=1,NegativeHeight=1)
    ActiveRegionSize=0.02
    Debug=(bDrawDebug=false,bDraw2D=false,DebugScale=25,DebugSegments=20,DebugThickness=0.25)
    DriverAxis=(X=0,Y=1,Z=-1)
    DriverItem=(Type=Bone,Name="pelvis")
    FalloffRegionScaleFactors=(PositiveWidth=1,NegativeWidth=1,PositiveHeight=1,NegativeHeight=1)
    FalloffSize=0.08
    OptionalParentItem=(Type=Bone,Name="root")
    RotationOffset=(X=60,Y=154,Z=17)
   )
   > &RigUnit_ModifyTransforms(
    ItemToModify=((Item=(Type=Bone,Name="dyn_hair_tair_1_r"),Transform=(Rotation=(X=0,Y=0,Z=0,W=1),Translation=(X=0,Y=0,Z=0),Scale3D=(X=1,Y=1,Z=1))))
    ItemToModify.0.Transform=&RigVMFunction_MathTransformFromSRT(Location=(X=-3.4,Y=0,Z=1), Rotation=(X=0,Y=10,Z=10), Scale=(X=1,Y=1,Z=1)).Transform
    Weight=&RigVMFunction_MathFloatMul(A=@n9.OutputParam, B=@n8.OutputParam) {template="Multiply::Execute(in A,in B,out Result)"}.Result
   )
   n10 = &RigUnit_SphericalPoseReader(
    ActiveRegionScaleFactors=(PositiveWidth=1,NegativeWidth=1,PositiveHeight=1,NegativeHeight=1)
    ActiveRegionSize=0.02
    Debug=(bDrawDebug=false,bDraw2D=false,DebugScale=25,DebugSegments=20,DebugThickness=0.25)
    DriverAxis=(X=0,Y=1,Z=-1)
    DriverItem=(Type=Bone,Name="upperarm_r")
    FalloffRegionScaleFactors=(PositiveWidth=1,NegativeWidth=1,PositiveHeight=1,NegativeHeight=1)
    FalloffSize=0.08
    OptionalParentItem=(Type=Bone)
    RotationOffset=(X=0,Y=105,Z=42)
    ExecutePin=@n3.True
   )
   > &RigUnit_ModifyTransforms(
    ItemToModify=((Item=(Type=Bone,Name="dyn_hair_tair_1_r"),Transform=(Rotation=(X=0,Y=0,Z=0,W=1),Translation=(X=0,Y=0,Z=0),Scale3D=(X=1,Y=1,Z=1))))
    ItemToModify.0.Transform=&RigVMFunction_MathTransformFromSRT(Location=(X=0,Y=0,Z=0), Rotation=(X=0,Y=0,Z=0), Scale=(X=1,Y=1,Z=1)).Transform
    Weight=@n10.OutputParam
   )
   n11 = &RigVMFunction_ControlFlowBranch(ExecuteContext=@RigVMFunction_ControlFlowBranch.True, Condition=@n2.Value)
   n12 = &RigUnit_SphericalPoseReader(
    ActiveRegionScaleFactors=(PositiveWidth=1,NegativeWidth=1,PositiveHeight=1,NegativeHeight=1)
    ActiveRegionSize=0.02
    Debug=(bDrawDebug=false,bDraw2D=false,DebugScale=25,DebugSegments=20,DebugThickness=0.25)
    DriverAxis=(X=0,Y=1,Z=-1)
    DriverItem=(Type=Bone,Name="upperarm_r")
    FalloffRegionScaleFactors=(PositiveWidth=1,NegativeWidth=1,PositiveHeight=1,NegativeHeight=1)
    FalloffSize=0.08
    OptionalParentItem=(Type=Bone)
    RotationOffset=(X=0,Y=105,Z=42)
    ExecutePin=@n11.False
   )
   > &RigUnit_ModifyTransforms(
    ItemToModify=((Item=(Type=Bone,Name="dyn_hair_tair_1_r"),Transform=(Rotation=(X=0,Y=0,Z=0,W=1),Translation=(X=0,Y=0,Z=0),Scale3D=(X=1,Y=1,Z=1))))
    ItemToModify.0.Transform=&RigVMFunction_MathTransformFromSRT(Location=(X=-2,Y=-3,Z=0), Rotation=(X=0,Y=-10,Z=10), Scale=(X=1,Y=1,Z=1)).Transform
    Weight=@n12.OutputParam
   )
   > &RigUnit_ModifyTransforms(
    ItemToModify=((Item=(Type=Bone,Name="dyn_head"),Transform=(Rotation=(X=0,Y=0,Z=0,W=1),Translation=(X=0,Y=0,Z=0),Scale3D=(X=1,Y=1,Z=1))))
    ItemToModify.0.Transform=&RigVMFunction_MathTransformFromSRT(Location=(X=0,Y=-1,Z=0), Rotation=(X=0,Y=0,Z=0), Scale=(X=1,Y=1,Z=1)).Transform
    Weight=@n12.OutputParam
   )
   > &RigUnit_ModifyTransforms(
    ItemToModify=((Item=(Type=Bone,Name="dyn_hair_tail_1_l"),Transform=(Rotation=(X=0,Y=0,Z=0,W=1),Translation=(X=0,Y=0,Z=0),Scale3D=(X=1,Y=1,Z=1))))
    ItemToModify.0.Transform=&RigVMFunction_MathTransformFromSRT(Location=(X=-3,Y=1,Z=1), Rotation=(X=0,Y=0,Z=5), Scale=(X=1,Y=1,Z=1)).Transform
    Weight=@n12.OutputParam
   )
   > n13 = &RigUnit_SphericalPoseReader(
    ActiveRegionScaleFactors=(PositiveWidth=1,NegativeWidth=1,PositiveHeight=1,NegativeHeight=1)
    ActiveRegionSize=0.04
    Debug=(bDrawDebug=false,bDraw2D=false,DebugScale=25,DebugSegments=20,DebugThickness=0.25)
    DriverAxis=(X=0,Y=1,Z=-1)
    DriverItem=(Type=Bone,Name="upperarm_r")
    FalloffRegionScaleFactors=(PositiveWidth=1,NegativeWidth=1,PositiveHeight=1,NegativeHeight=1)
    FalloffSize=0.13
    OptionalParentItem=(Type=Bone,Name="pelvis")
    RotationOffset=(X=50,Y=120,Z=17)
   )
   > n14 = &RigUnit_SphericalPoseReader(
    ActiveRegionScaleFactors=(PositiveWidth=1,NegativeWidth=2,PositiveHeight=1,NegativeHeight=1)
    ActiveRegionSize=0.02
    Debug=(bDrawDebug=false,bDraw2D=false,DebugScale=25,DebugSegments=20,DebugThickness=0.25)
    DriverAxis=(X=0,Y=1,Z=-1)
    DriverItem=(Type=Bone,Name="pelvis")
    FalloffRegionScaleFactors=(PositiveWidth=1,NegativeWidth=1,PositiveHeight=1,NegativeHeight=1)
    FalloffSize=0.08
    OptionalParentItem=(Type=Bone,Name="root")
    RotationOffset=(X=60,Y=154,Z=17)
   )
   > &RigUnit_ModifyTransforms(
    ItemToModify=((Item=(Type=Bone,Name="dyn_hair_tair_1_r"),Transform=(Rotation=(X=0,Y=0,Z=0,W=1),Translation=(X=0,Y=0,Z=0),Scale3D=(X=1,Y=1,Z=1))))
    ItemToModify.0.Transform=&RigVMFunction_MathTransformFromSRT(Location=(X=-5,Y=2,Z=1), Rotation=(X=0,Y=0,Z=10), Scale=(X=1,Y=1,Z=1)).Transform
    Weight=&RigVMFunction_MathFloatMul(B=@n13.OutputParam, A=@n14.OutputParam) {template="Multiply::Execute(in A,in B,out Result)"}.Result
   )
   > n15 = &RigUnit_SphericalPoseReader(
    ActiveRegionScaleFactors=(PositiveWidth=4,NegativeWidth=1,PositiveHeight=6,NegativeHeight=1)
    ActiveRegionSize=0.035
    Debug=(bDrawDebug=false,bDraw2D=false,DebugScale=25,DebugSegments=20,DebugThickness=0.25)
    DriverAxis=(X=1,Y=0,Z=0)
    DriverItem=(Type=Bone,Name="head")
    FalloffRegionScaleFactors=(PositiveWidth=1,NegativeWidth=1,PositiveHeight=1,NegativeHeight=0.3)
    FalloffSize=0.3
    OptionalParentItem=(Type=Bone)
    RotationOffset=(X=0,Y=90,Z=160)
   )
   > &RigUnit_ModifyTransforms(
    ItemToModify=((Item=(Type=Bone,Name="dyn_hair_tair_1_r"),Transform=(Rotation=(X=0,Y=0,Z=0,W=1),Translation=(X=0,Y=0,Z=0),Scale3D=(X=1,Y=1,Z=1))))
    ItemToModify.0.Transform=&RigVMFunction_MathTransformFromSRT(Location=(X=-0.5,Y=-1,Z=0), Rotation=(X=0,Y=0,Z=5), Scale=(X=1,Y=1,Z=1)).Transform
    Weight=@n15.OutputParam
   )
   > &RigUnit_ModifyTransforms(
    ItemToModify=((Item=(Type=Bone,Name="dyn_hair_tail_1_l"),Transform=(Rotation=(X=0,Y=0,Z=0,W=1),Translation=(X=0,Y=0,Z=0),Scale3D=(X=1,Y=1,Z=1))))
    ItemToModify.0.Transform=&RigVMFunction_MathTransformFromSRT(Location=(X=-2,Y=2,Z=0), Rotation=(X=0,Y=0,Z=0), Scale=(X=1,Y=1,Z=1)).Transform
    Weight=@n15.OutputParam
   )
   > n16 = &RigUnit_SphericalPoseReader(
    ActiveRegionScaleFactors=(PositiveWidth=1,NegativeWidth=1,PositiveHeight=1,NegativeHeight=1)
    ActiveRegionSize=0.035
    Debug=(bDrawDebug=false,bDraw2D=false,DebugScale=25,DebugSegments=20,DebugThickness=0.25)
    DriverAxis=(X=0,Y=1,Z=0)
    DriverItem=(Type=Bone,Name="spine_05")
    FalloffRegionScaleFactors=(PositiveWidth=1,NegativeWidth=1,PositiveHeight=1,NegativeHeight=0.5)
    FalloffSize=0.3
    OptionalParentItem=(Type=Bone,Name="root")
    RotationOffset=(X=10,Y=0,Z=90)
   )
   > &RigUnit_ModifyTransforms(
    ItemToModify=((Item=(Type=Bone,Name="dyn_hair_tair_1_r"),Transform=(Rotation=(X=0,Y=0,Z=0,W=1),Translation=(X=0,Y=0,Z=0),Scale3D=(X=1,Y=1,Z=1))))
    ItemToModify.0.Transform=&RigVMFunction_MathTransformFromSRT(Location=(X=0,Y=0,Z=0), Rotation=(X=0,Y=22,Z=0), Scale=(X=1,Y=1,Z=1)).Transform
    Weight=@n16.OutputParam
   )
   n17 = &RigUnit_SphericalPoseReader(
    ActiveRegionScaleFactors=(PositiveWidth=1,NegativeWidth=1,PositiveHeight=1,NegativeHeight=1)
    ActiveRegionSize=0.02
    Debug=(bDrawDebug=false,bDraw2D=false,DebugScale=25,DebugSegments=20,DebugThickness=0.25)
    DriverAxis=(X=0,Y=1,Z=-1)
    DriverItem=(Type=Bone,Name="upperarm_r")
    FalloffRegionScaleFactors=(PositiveWidth=1,NegativeWidth=1,PositiveHeight=1,NegativeHeight=1)
    FalloffSize=0.08
    OptionalParentItem=(Type=Bone)
    RotationOffset=(X=0,Y=105,Z=42)
    ExecutePin=@n11.True
   )
   > &RigUnit_ModifyTransforms(
    ItemToModify=((Item=(Type=Bone,Name="dyn_hair_tair_1_r"),Transform=(Rotation=(X=0,Y=0,Z=0,W=1),Translation=(X=0,Y=0,Z=0),Scale3D=(X=1,Y=1,Z=1))))
    ItemToModify.0.Transform=&RigVMFunction_MathTransformFromSRT(Location=(X=0,Y=0,Z=0), Rotation=(X=0,Y=0,Z=0), Scale=(X=1,Y=1,Z=1)).Transform
    Weight=@n17.OutputParam
   )
   RigUnit_SequenceAggregate = &RigVMFunction_Sequence(ExecuteContext=())
   n18 = &RigUnit_SphericalPoseReader(
    ActiveRegionScaleFactors=(PositiveWidth=1.5,NegativeWidth=2,PositiveHeight=2,NegativeHeight=1)
    ActiveRegionSize=0.035
    Debug=(bDrawDebug=false,bDraw2D=false,DebugScale=25,DebugSegments=20,DebugThickness=0.25)
    DriverAxis=(X=1,Y=0,Z=0)
    DriverItem=(Type=Bone,Name="head")
    FalloffRegionScaleFactors=(PositiveWidth=1,NegativeWidth=1,PositiveHeight=1,NegativeHeight=1)
    FalloffSize=0.1
    OptionalParentItem=(Type=Bone)
    RotationOffset=(X=0,Y=70,Z=155)
    ExecutePin=@RigUnit_SequenceAggregate.A
   )
   > &RigUnit_ModifyTransforms(
    ItemToModify=((Item=(Type=Bone,Name="dyn_hair_tair_1_r"),Transform=(Rotation=(X=0,Y=0,Z=0,W=1),Translation=(X=0,Y=0,Z=0),Scale3D=(X=1,Y=1,Z=1))))
    ItemToModify.0.Transform=&RigVMFunction_MathTransformFromSRT(Location=(X=2,Y=0,Z=-0.5), Rotation=(X=0,Y=0,Z=-15), Scale=(X=1,Y=1,Z=1)).Transform
    Weight=@n18.OutputParam
   )
   > &RigUnit_ModifyTransforms(
    ItemToModify=((Item=(Type=Bone,Name="dyn_hair_tail_1_l"),Transform=(Rotation=(X=0,Y=0,Z=0,W=1),Translation=(X=0,Y=0,Z=0),Scale3D=(X=1,Y=1,Z=1))))
    ItemToModify.0.Transform=&RigVMFunction_MathTransformFromSRT(Location=(X=0,Y=0,Z=0), Rotation=(X=0,Y=-20,Z=0), Scale=(X=1,Y=1,Z=1)).Transform
    Weight=@n18.OutputParam
   )
   > n19 = &RigUnit_SphericalPoseReader(
    ActiveRegionScaleFactors=(PositiveWidth=1,NegativeWidth=1.5,PositiveHeight=1.8,NegativeHeight=1)
    ActiveRegionSize=0.02
    Debug=(bDrawDebug=false,bDraw2D=false,DebugScale=25,DebugSegments=20,DebugThickness=0.25)
    DriverAxis=(X=0,Y=1,Z=-1)
    DriverItem=(Type=Bone,Name="upperarm_r")
    FalloffRegionScaleFactors=(PositiveWidth=1,NegativeWidth=2,PositiveHeight=1.4,NegativeHeight=1)
    FalloffSize=0.08
    OptionalParentItem=(Type=Bone)
    RotationOffset=(X=0,Y=105,Z=42)
   )
   > &RigUnit_ModifyTransforms(
    ItemToModify=((Item=(Type=Bone,Name="dyn_hair_tail_1_l"),Transform=(Rotation=(X=0,Y=0,Z=0,W=1),Translation=(X=0,Y=0,Z=0),Scale3D=(X=1,Y=1,Z=1))))
    ItemToModify.0.Transform=&RigVMFunction_MathTransformFromSRT(Location=(X=-3,Y=3,Z=0), Rotation=(X=0,Y=0,Z=-40), Scale=(X=1,Y=1,Z=1)).Transform
    Weight=@n19.OutputParam
   )
   > &RigUnit_ModifyTransforms(
    ItemToModify=((Item=(Type=Bone,Name="dyn_hair_tair_1_r"),Transform=(Rotation=(X=0,Y=0,Z=0,W=1),Translation=(X=0,Y=0,Z=0),Scale3D=(X=1,Y=1,Z=1))))
    ItemToModify.0.Transform=&RigVMFunction_MathTransformFromSRT(Location=(X=2,Y=-4,Z=-1), Rotation=(X=0,Y=0,Z=-30), Scale=(X=1,Y=1,Z=1)).Transform
    Weight=@n19.OutputParam
   )
   > &RigUnit_ModifyTransforms(
    ItemToModify=((Item=(Type=Bone,Name="dyn_head"),Transform=(Rotation=(X=0,Y=0,Z=0,W=1),Translation=(X=0,Y=0,Z=0),Scale3D=(X=1,Y=1,Z=1))))
    ItemToModify.0.Transform=&RigVMFunction_MathTransformFromSRT(Location=(X=0,Y=-1,Z=0), Rotation=(X=0,Y=0,Z=20), Scale=(X=1,Y=1,Z=1)).Transform
    Weight=@n19.OutputParam
   )
   n20 = &RigUnit_SphericalPoseReader(
    ActiveRegionScaleFactors=(PositiveWidth=1,NegativeWidth=1,PositiveHeight=1,NegativeHeight=1)
    ActiveRegionSize=0.02
    Debug=(bDrawDebug=false,bDraw2D=false,DebugScale=25,DebugSegments=20,DebugThickness=0.25)
    DriverAxis=(X=0,Y=1,Z=-1)
    DriverItem=(Type=Bone,Name="upperarm_r")
    FalloffRegionScaleFactors=(PositiveWidth=1,NegativeWidth=1,PositiveHeight=1,NegativeHeight=1)
    FalloffSize=0.08
    OptionalParentItem=(Type=Bone)
    RotationOffset=(X=0,Y=105,Z=0)
    ExecutePin=@RigUnit_SequenceAggregate.B
   )
   > &RigUnit_ModifyTransforms(
    ItemToModify=((Item=(Type=Bone,Name="dyn_hair_tair_1_r"),Transform=(Rotation=(X=0,Y=0,Z=0,W=1),Translation=(X=0,Y=0,Z=0),Scale3D=(X=1,Y=1,Z=1))))
    Weight=1
    ItemToModify.0.Transform=&RigVMFunction_MathTransformFromSRT(Location=(X=0,Y=0,Z=-1), Rotation=(X=0,Y=0,Z=-10), Scale=(X=1,Y=1,Z=1)).Transform
   )
   > &RigUnit_ModifyTransforms(
    ItemToModify=((Item=(Type=Bone,Name="dyn_hair_tail_1_l"),Transform=(Rotation=(X=0,Y=0,Z=0,W=1),Translation=(X=0,Y=0,Z=0),Scale3D=(X=1,Y=1,Z=1))))
    ItemToModify.0.Transform=&RigVMFunction_MathTransformFromSRT(Location=(X=0,Y=0,Z=0.5), Rotation=(X=0,Y=10,Z=-10), Scale=(X=1,Y=1,Z=1)).Transform
    Weight=@n20.OutputParam
   )
   > n21 = &RigUnit_SphericalPoseReader(
    ActiveRegionScaleFactors=(PositiveWidth=1,NegativeWidth=1,PositiveHeight=1,NegativeHeight=3)
    ActiveRegionSize=0.02
    Debug=(bDrawDebug=false,bDraw2D=false,DebugScale=25,DebugSegments=20,DebugThickness=0.25)
    DriverAxis=(X=0,Y=1,Z=-1)
    DriverItem=(Type=Bone,Name="upperarm_r")
    FalloffRegionScaleFactors=(PositiveWidth=1,NegativeWidth=1.3,PositiveHeight=1,NegativeHeight=5)
    FalloffSize=0.08
    OptionalParentItem=(Type=Bone)
    RotationOffset=(X=0,Y=84,Z=10)
   )
   > &RigUnit_ModifyTransforms(
    ItemToModify=((Item=(Type=Bone,Name="dyn_hair_tail_1_l"),Transform=(Rotation=(X=0,Y=0,Z=0,W=1),Translation=(X=0,Y=0,Z=0),Scale3D=(X=1,Y=1,Z=1))))
    ItemToModify.0.Transform=&RigVMFunction_MathTransformFromSRT(Location=(X=0,Y=0,Z=0), Rotation=(X=0,Y=10,Z=10), Scale=(X=1,Y=1,Z=1)).Transform
    Weight=@n21.OutputParam
   )
   > &RigUnit_ModifyTransforms(
    ItemToModify=((Item=(Type=Bone,Name="dyn_hair_tair_1_r"),Transform=(Rotation=(X=0,Y=0,Z=0,W=1),Translation=(X=0,Y=0,Z=0),Scale3D=(X=1,Y=1,Z=1))))
    ItemToModify.0.Transform=&RigVMFunction_MathTransformFromSRT(Location=(X=0,Y=0,Z=-0.5), Rotation=(X=0,Y=0,Z=-10), Scale=(X=1,Y=1,Z=1)).Transform
    Weight=@n21.OutputParam
   )
   > n22 = &RigUnit_SphericalPoseReader(
    ActiveRegionScaleFactors=(PositiveWidth=1,NegativeWidth=1,PositiveHeight=1,NegativeHeight=1)
    ActiveRegionSize=0.03
    Debug=(bDrawDebug=false,DebugScale=25,DebugSegments=20,DebugThickness=0.25)
    DriverAxis=(X=0,Y=1,Z=0)
    DriverItem=(Type=Bone,Name="head")
    FalloffRegionScaleFactors=(PositiveWidth=1,NegativeWidth=1,PositiveHeight=1,NegativeHeight=1)
    FalloffSize=0.1
    OptionalParentItem=(Type=Bone,Name="spine_05")
    RotationOffset=(X=135,Y=0,Z=0)
   )
   RigVMFunction_MathTransformFromSRT = &RigVMFunction_MathTransformFromSRT(Location=(X=0,Y=0,Z=0), Rotation=(X=0,Y=0,Z=20), Scale=(X=1,Y=1,Z=1))
   > RigUnit_SetRelativeTransformForItem = /Script/ControlRig.RigUnit_SetRelativeTransformForItem(
    Child=(Type=Bone,Name="dyn_hair_tair_1_r")
    Parent=(Type=Bone,Name="dyn_hair_tair_1_r")
    bParentInitial=False
    bPropagateToChildren=True
    Value=@RigVMFunction_MathTransformFromSRT.Transform
    Weight=@n22.OutputParam
   ) {
    is=struct RigVMUnitNode
    template="Set Relative Transform::Execute(in Child,in Parent,in bParentInitial,in Value,in Weight,in bPropagateToChildren)"
   }
   n23 = &RigUnit_SphericalPoseReader(
    ActiveRegionScaleFactors=(PositiveWidth=1.5,NegativeWidth=2,PositiveHeight=2,NegativeHeight=1)
    ActiveRegionSize=0.035
    Debug=(bDrawDebug=true,bDraw2D=false,DebugScale=25,DebugSegments=20,DebugThickness=0.25)
    DriverAxis=(X=1,Y=0,Z=0)
    DriverItem=(Type=Bone,Name="head")
    ExecutePin=()
    FalloffRegionScaleFactors=(PositiveWidth=1,NegativeWidth=1,PositiveHeight=1,NegativeHeight=1)
    FalloffSize=0.1
    OptionalParentItem=(Type=Bone)
    RotationOffset=(X=0,Y=70,Z=155)
   )
   > &RigUnit_ModifyTransforms(
    ItemToModify=((Item=(Type=Bone,Name="dyn_hair_tair_1_r"),Transform=(Rotation=(X=0,Y=0,Z=0,W=1),Translation=(X=0,Y=0,Z=0),Scale3D=(X=1,Y=1,Z=1))))
    ItemToModify.0.Transform=&RigVMFunction_MathTransformFromSRT(Location=(X=1.5,Y=0,Z=-0.7), Rotation=(X=0,Y=0,Z=-15), Scale=(X=1,Y=1,Z=1)).Transform
    Weight=@n23.OutputParam
   )
   > &RigUnit_ModifyTransforms(
    ItemToModify=((Item=(Type=Bone,Name="dyn_hair_tail_1_l"),Transform=(Rotation=(X=0,Y=0,Z=0,W=1),Translation=(X=0,Y=0,Z=0),Scale3D=(X=1,Y=1,Z=1))))
    ItemToModify.0.Transform=&RigVMFunction_MathTransformFromSRT(Location=(X=-1,Y=1,Z=0), Rotation=(X=0,Y=-10,Z=-20), Scale=(X=1,Y=1,Z=1)).Transform
    Weight=@n23.OutputParam
   )
   n24 = &RigUnit_SphericalPoseReader(
    ActiveRegionScaleFactors=(PositiveWidth=4,NegativeWidth=1,PositiveHeight=6,NegativeHeight=1)
    ActiveRegionSize=0.035
    Debug=(bDrawDebug=true,bDraw2D=false,DebugScale=25,DebugSegments=20,DebugThickness=0.25)
    DriverAxis=(X=1,Y=0,Z=0)
    DriverItem=(Type=Bone,Name="head")
    ExecutePin=()
    FalloffRegionScaleFactors=(PositiveWidth=1,NegativeWidth=1,PositiveHeight=1,NegativeHeight=0.5)
    FalloffSize=0.3
    OptionalParentItem=(Type=Bone)
    RotationOffset=(X=7,Y=-90,Z=0)
   )
   > &RigUnit_ModifyTransforms(
    ItemToModify=((Item=(Type=Bone,Name="dyn_hair_tair_1_r"),Transform=(Rotation=(X=0,Y=0,Z=0,W=1),Translation=(X=0,Y=0,Z=0),Scale3D=(X=1,Y=1,Z=1))))
    ItemToModify.0.Transform=&RigVMFunction_MathTransformFromSRT(Location=(X=0,Y=2,Z=0), Rotation=(X=0,Y=22,Z=0), Scale=(X=1,Y=1,Z=1)).Transform
    Weight=&RigVMFunction_MathFloatMul(A=1, B=@n24.OutputParam) {template="Multiply::Execute(in A,in B,out Result)"}.Result
   )
   n25 = &RigUnit_SphericalPoseReader(
    ActiveRegionScaleFactors=(PositiveWidth=1,NegativeWidth=1,PositiveHeight=1,NegativeHeight=1)
    ActiveRegionSize=0.02
    Debug=(bDrawDebug=false,bDraw2D=false,DebugScale=25,DebugSegments=20,DebugThickness=0.25)
    DriverAxis=(X=0,Y=1,Z=-1)
    DriverItem=(Type=Bone,Name="upperarm_r")
    ExecutePin=()
    FalloffRegionScaleFactors=(PositiveWidth=1,NegativeWidth=1,PositiveHeight=1,NegativeHeight=1)
    FalloffSize=0.08
    OptionalParentItem=(Type=Bone)
    RotationOffset=(X=0,Y=105,Z=42)
   )
   > &RigUnit_ModifyTransforms(
    ItemToModify=((Item=(Type=Bone,Name="dyn_hair_tair_1_r"),Transform=(Rotation=(X=0,Y=0,Z=0,W=1),Translation=(X=0,Y=0,Z=0),Scale3D=(X=1,Y=1,Z=1))))
    ItemToModify.0.Transform=&RigVMFunction_MathTransformFromSRT(Location=(X=-2,Y=-3,Z=0), Rotation=(X=0,Y=-10,Z=15), Scale=(X=1,Y=1,Z=1)).Transform
    Weight=@n25.OutputParam
   )
   > &RigUnit_ModifyTransforms(
    ItemToModify=((Item=(Type=Bone,Name="dyn_head"),Transform=(Rotation=(X=0,Y=0,Z=0,W=1),Translation=(X=0,Y=0,Z=0),Scale3D=(X=1,Y=1,Z=1))))
    ItemToModify.0.Transform=&RigVMFunction_MathTransformFromSRT(Location=(X=0,Y=-1,Z=0), Rotation=(X=0,Y=0,Z=0), Scale=(X=1,Y=1,Z=1)).Transform
    Weight=@n25.OutputParam
   )
   > &RigUnit_ModifyTransforms(
    ItemToModify=((Item=(Type=Bone,Name="dyn_hair_tail_1_l"),Transform=(Rotation=(X=0,Y=0,Z=0,W=1),Translation=(X=0,Y=0,Z=0),Scale3D=(X=1,Y=1,Z=1))))
    ItemToModify.0.Transform=&RigVMFunction_MathTransformFromSRT(Location=(X=-3,Y=1,Z=1), Rotation=(X=0,Y=0,Z=5), Scale=(X=1,Y=1,Z=1)).Transform
    Weight=@n25.OutputParam
   )
   > n26 = &RigUnit_SphericalPoseReader(
    ActiveRegionScaleFactors=(PositiveWidth=1,NegativeWidth=1,PositiveHeight=1,NegativeHeight=1)
    ActiveRegionSize=0.04
    Debug=(bDrawDebug=false,bDraw2D=false,DebugScale=25,DebugSegments=20,DebugThickness=0.25)
    DriverAxis=(X=0,Y=1,Z=-1)
    DriverItem=(Type=Bone,Name="upperarm_r")
    FalloffRegionScaleFactors=(PositiveWidth=1,NegativeWidth=1,PositiveHeight=1,NegativeHeight=1)
    FalloffSize=0.13
    OptionalParentItem=(Type=Bone,Name="pelvis")
    RotationOffset=(X=50,Y=120,Z=17)
   )
   > n27 = &RigUnit_SphericalPoseReader(
    ActiveRegionScaleFactors=(PositiveWidth=1,NegativeWidth=2,PositiveHeight=1,NegativeHeight=1)
    ActiveRegionSize=0.02
    Debug=(bDrawDebug=false,bDraw2D=false,DebugScale=25,DebugSegments=20,DebugThickness=0.25)
    DriverAxis=(X=0,Y=1,Z=-1)
    DriverItem=(Type=Bone,Name="pelvis")
    FalloffRegionScaleFactors=(PositiveWidth=1,NegativeWidth=1,PositiveHeight=1,NegativeHeight=1)
    FalloffSize=0.08
    OptionalParentItem=(Type=Bone,Name="root")
    RotationOffset=(X=60,Y=154,Z=17)
   )
   > &RigUnit_ModifyTransforms(
    ItemToModify=((Item=(Type=Bone,Name="dyn_hair_tair_1_r"),Transform=(Rotation=(X=0,Y=0,Z=0,W=1),Translation=(X=0,Y=0,Z=0),Scale3D=(X=1,Y=1,Z=1))))
    ItemToModify.0.Transform=&RigVMFunction_MathTransformFromSRT(Location=(X=-3.4,Y=0,Z=1), Rotation=(X=0,Y=0,Z=10), Scale=(X=1,Y=1,Z=1)).Transform
    Weight=&RigVMFunction_MathFloatMul(A=@n27.OutputParam, B=@n26.OutputParam) {template="Multiply::Execute(in A,in B,out Result)"}.Result
   )
   > n28 = &RigUnit_SphericalPoseReader(
    ActiveRegionScaleFactors=(PositiveWidth=4,NegativeWidth=1,PositiveHeight=6,NegativeHeight=1)
    ActiveRegionSize=0.035
    Debug=(bDrawDebug=false,bDraw2D=false,DebugScale=25,DebugSegments=20,DebugThickness=0.25)
    DriverAxis=(X=1,Y=0,Z=0)
    DriverItem=(Type=Bone,Name="head")
    FalloffRegionScaleFactors=(PositiveWidth=1,NegativeWidth=1,PositiveHeight=1,NegativeHeight=0.3)
    FalloffSize=0.3
    OptionalParentItem=(Type=Bone)
    RotationOffset=(X=0,Y=90,Z=160)
   )
   > &RigUnit_ModifyTransforms(
    ItemToModify=((Item=(Type=Bone,Name="dyn_hair_tair_1_r"),Transform=(Rotation=(X=0,Y=0,Z=0,W=1),Translation=(X=0,Y=0,Z=0),Scale3D=(X=1,Y=1,Z=1))))
    ItemToModify.0.Transform=&RigVMFunction_MathTransformFromSRT(Location=(X=-2,Y=-2,Z=0), Rotation=(X=0,Y=0,Z=0), Scale=(X=1,Y=1,Z=1)).Transform
    Weight=@n28.OutputParam
   )
   > &RigUnit_ModifyTransforms(
    ItemToModify=((Item=(Type=Bone,Name="dyn_hair_tail_1_l"),Transform=(Rotation=(X=0,Y=0,Z=0,W=1),Translation=(X=0,Y=0,Z=0),Scale3D=(X=1,Y=1,Z=1))))
    ItemToModify.0.Transform=&RigVMFunction_MathTransformFromSRT(Location=(X=-2,Y=2,Z=0), Rotation=(X=0,Y=0,Z=0), Scale=(X=1,Y=1,Z=1)).Transform
    Weight=@n28.OutputParam
   )
   > n29 = &RigUnit_SphericalPoseReader(
    ActiveRegionScaleFactors=(PositiveWidth=4,NegativeWidth=1,PositiveHeight=6,NegativeHeight=1)
    ActiveRegionSize=0.035
    Debug=(bDrawDebug=false,bDraw2D=false,DebugScale=25,DebugSegments=20,DebugThickness=0.25)
    DriverAxis=(X=0,Y=1,Z=0)
    DriverItem=(Type=Bone,Name="spine_05")
    FalloffRegionScaleFactors=(PositiveWidth=1,NegativeWidth=1,PositiveHeight=1,NegativeHeight=0.5)
    FalloffSize=0.3
    OptionalParentItem=(Type=Bone,Name="root")
    RotationOffset=(X=10,Y=0,Z=90)
   )
   > &RigUnit_ModifyTransforms(
    ItemToModify=((Item=(Type=Bone,Name="dyn_hair_tair_1_r"),Transform=(Rotation=(X=0,Y=0,Z=0,W=1),Translation=(X=0,Y=0,Z=0),Scale3D=(X=1,Y=1,Z=1))))
    ItemToModify.0.Transform=&RigVMFunction_MathTransformFromSRT(Location=(X=3,Y=4,Z=0), Rotation=(X=0,Y=20,Z=0), Scale=(X=1,Y=1,Z=1)).Transform
    Weight=@n29.OutputParam
   )
   > &RigUnit_ModifyTransforms(
    ItemToModify=((Item=(Type=Bone,Name="dyn_hair_tail_1_l"),Transform=(Rotation=(X=0,Y=0,Z=0,W=1),Translation=(X=0,Y=0,Z=0),Scale3D=(X=1,Y=1,Z=1))))
    ItemToModify.0.Transform=&RigVMFunction_MathTransformFromSRT(Location=(X=0,Y=0,Z=0), Rotation=(X=0,Y=0,Z=0), Scale=(X=1,Y=1,Z=1)).Transform
    Weight=@n29.OutputParam
   )
   n30 = &RigUnit_SphericalPoseReader(
    ActiveRegionScaleFactors=(PositiveWidth=1,NegativeWidth=1,PositiveHeight=1,NegativeHeight=1)
    ActiveRegionSize=0.02
    Debug=(bDrawDebug=false,bDraw2D=false,DebugScale=25,DebugSegments=20,DebugThickness=0.25)
    DriverAxis=(X=0,Y=1,Z=-1)
    DriverItem=(Type=Bone,Name="upperarm_r")
    ExecutePin=()
    FalloffRegionScaleFactors=(PositiveWidth=1,NegativeWidth=1,PositiveHeight=1,NegativeHeight=1)
    FalloffSize=0.08
    OptionalParentItem=(Type=Bone)
    RotationOffset=(X=0,Y=105,Z=42)
   )
   > &RigUnit_ModifyTransforms(
    ItemToModify=((Item=(Type=Bone,Name="dyn_hair_tair_1_r"),Transform=(Rotation=(X=0,Y=0,Z=0,W=1),Translation=(X=0,Y=0,Z=0),Scale3D=(X=1,Y=1,Z=1))))
    ItemToModify.0.Transform=&RigVMFunction_MathTransformFromSRT(Location=(X=-2,Y=-3,Z=0), Rotation=(X=0,Y=-10,Z=15), Scale=(X=1,Y=1,Z=1)).Transform
    Weight=@n30.OutputParam
   )
   > &RigUnit_ModifyTransforms(
    ItemToModify=((Item=(Type=Bone,Name="dyn_head"),Transform=(Rotation=(X=0,Y=0,Z=0,W=1),Translation=(X=0,Y=0,Z=0),Scale3D=(X=1,Y=1,Z=1))))
    ItemToModify.0.Transform=&RigVMFunction_MathTransformFromSRT(Location=(X=0,Y=-1,Z=0), Rotation=(X=0,Y=0,Z=0), Scale=(X=1,Y=1,Z=1)).Transform
    Weight=@n30.OutputParam
   )
   > &RigUnit_ModifyTransforms(
    ItemToModify=((Item=(Type=Bone,Name="dyn_hair_tail_1_l"),Transform=(Rotation=(X=0,Y=0,Z=0,W=1),Translation=(X=0,Y=0,Z=0),Scale3D=(X=1,Y=1,Z=1))))
    ItemToModify.0.Transform=&RigVMFunction_MathTransformFromSRT(Location=(X=-3,Y=1,Z=1), Rotation=(X=0,Y=0,Z=5), Scale=(X=1,Y=1,Z=1)).Transform
    Weight=@n30.OutputParam
   )
   &RigVMFunction_MathTransformFromSRT(Location=(X=-3,Y=-2,Z=0), Rotation=(X=0,Y=22,Z=0), Scale=(X=1,Y=1,Z=1))
   &RigUnit_ModifyTransforms(
    ExecutePin=()
    ItemToModify=((Item=(Type=Bone,Name="dyn_hair_tail_1_l"),Transform=(Rotation=(X=0,Y=0,Z=0,W=1),Translation=(X=0,Y=0,Z=0),Scale3D=(X=1,Y=1,Z=1))))
    Weight=1
    ItemToModify.0.Transform=&RigVMFunction_MathTransformFromSRT(Location=(X=-1,Y=2,Z=0), Rotation=(X=0,Y=0,Z=-20), Scale=(X=1,Y=1,Z=1)).Transform
   )
   &RigUnit_ModifyTransforms(
    ExecutePin=()
    ItemToModify=((Item=(Type=Bone,Name="dyn_hair_tail_1_l"),Transform=(Rotation=(X=0,Y=0,Z=0,W=1),Translation=(X=0,Y=0,Z=0),Scale3D=(X=1,Y=1,Z=1))))
    ItemToModify.0.Transform=&RigVMFunction_MathTransformFromSRT(Location=(X=0,Y=0,Z=0), Rotation=(X=0,Y=0,Z=0), Scale=(X=1,Y=1,Z=1)).Transform
    Weight=@n16.OutputParam
   )
   &RigUnit_ModifyTransforms(
    ExecutePin=()
    ItemToModify=((Item=(Type=Bone,Name="dyn_head"),Transform=(Rotation=(X=0,Y=0,Z=0,W=1),Translation=(X=0,Y=0,Z=0),Scale3D=(X=1,Y=1,Z=1))))
    Weight=1
    ItemToModify.0.Transform=&RigVMFunction_MathTransformFromSRT(Location=(X=0,Y=-1,Z=0), Rotation=(X=0,Y=0,Z=20), Scale=(X=1,Y=1,Z=1)).Transform
   )
   &RigUnit_ModifyTransforms(
    ExecutePin=()
    ItemToModify=((Item=(Type=Bone,Name="dyn_hair_tair_1_r"),Transform=(Rotation=(X=0,Y=0,Z=0,W=1),Translation=(X=0,Y=0,Z=0),Scale3D=(X=1,Y=1,Z=1))))
    Weight=1
    ItemToModify.0.Transform=&RigVMFunction_MathTransformFromSRT(Location=(X=0,Y=0,Z=0), Rotation=(X=0,Y=5,Z=10), Scale=(X=1,Y=1,Z=1)).Transform
   )
   Multiply = &RigVMFunction_MathDoubleMul(A=@n19.OutputParam, B=@n18.OutputParam)
   &RigUnit_SphericalPoseReader(
    ActiveRegionScaleFactors=(PositiveWidth=1,NegativeWidth=1,PositiveHeight=1,NegativeHeight=1)
    ActiveRegionSize=0.02
    Debug=(bDrawDebug=true,bDraw2D=false,DebugScale=25,DebugSegments=20,DebugThickness=0.25)
    DriverAxis=(X=0,Y=1,Z=-1)
    DriverItem=(Type=Bone,Name="upperarm_r")
    ExecutePin=()
    FalloffRegionScaleFactors=(PositiveWidth=1,NegativeWidth=1,PositiveHeight=1,NegativeHeight=0.5)
    FalloffSize=0.08
    OptionalParentItem=(Type=Bone,Name="clavicle_l")
    RotationOffset=(X=178,Y=0,Z=90)
   )
   &RigUnit_SphericalPoseReader(
    ActiveRegionScaleFactors=(PositiveWidth=1,NegativeWidth=1,PositiveHeight=1,NegativeHeight=1)
    ActiveRegionSize=0.02
    Debug=(bDrawDebug=true,bDraw2D=false,DebugScale=25,DebugSegments=20,DebugThickness=0.25)
    DriverAxis=(X=0,Y=1,Z=-1)
    DriverItem=(Type=Bone,Name="upperarm_r")
    ExecutePin=()
    FalloffRegionScaleFactors=(PositiveWidth=1,NegativeWidth=1,PositiveHeight=1,NegativeHeight=1)
    FalloffSize=0.08
    OptionalParentItem=(Type=Bone)
    RotationOffset=(X=0,Y=105,Z=0)
   )
  ]}
]
hierarchy [
 root {loc=<0 0 0>, rot=<0 0 0 1>}
 pelvis {parent=root, loc=<0 -0.5301 98.0571>, rot=<0.0127 0.707 -0.0127 -0.707>}
 spine_01 {parent=pelvis, loc=<6.1208 -0.9328 0>, rot=<0 0 0.034 -0.9994>}
 spine_02 {parent=spine_01, loc=<6.5996 -0.2095 0>, rot=<0 0 0.0238 -0.9997>}
 spine_03 {parent=spine_02, loc=<6.4914 -0.1237 0>, rot=<0 0 0.0149 -0.9999>}
 spine_04 {parent=spine_03, loc=<6.9598 0.0794 0>, rot=<0 0 -0.0132 -0.9999>}
 spine_05 {parent=spine_04, loc=<6.9987 0 0>, rot=<0 0 -0.0015 -1>}
 neck_01 {parent=spine_05, loc=<11.3138 0.1529 0>, rot=<0 0 -0.2003 -0.9797>}
 neck_02 {parent=neck_01, loc=<4.5627 0 0>, rot=<0 0 0.0504 -0.9987>}
 head {parent=neck_02, loc=<5.3914 0 0>, rot=<0 0 0.115 -0.9934>}
 dyn_hair_tair_1_r {parent=head, loc=<8.535 -4.036 2.641>, rot=<0.1389 -0.0119 0.9859 -0.0925>}
 dyn_hair_tair_1_r1 {parent=dyn_hair_tair_1_r, loc=<0 0 0>, rot=<0 0 0 1>}
 dyn_hair_tair_2_r {parent=dyn_hair_tair_1_r1, loc=<12.0225 0 0>, rot=<0.5533 0.2217 -0.1469 -0.7894>}
 dyn_hair_tail_1_l {parent=head, loc=<8.5346 -4.0356 -2.6412>, rot=<0.0925 -0.9859 0.0119 -0.1389>}
 dyn_hair_tail_1_l1 {parent=dyn_hair_tail_1_l, loc=<0 0 0>, rot=<0 0 0 1>}
 dyn_hair_tail_2_l {parent=dyn_hair_tail_1_l1, loc=<12.0228 0 0>, rot=<0.8225 -0.1499 0.178 -0.5189>}
 dyn_hair_3_r {parent=head, loc=<8.9693 8.1006 4.0875>, rot=<0.07 -0.9905 -0.0073 0.1177>}
 dyn_hair_3_r_1 {parent=dyn_hair_3_r, loc=<4.4299 0 0>, rot=<-0.6602 0.0595 0.22 -0.7157>}
 dyn_hair_3_r_2 {parent=dyn_hair_3_r_1, loc=<5.1272 0 0>, rot=<-0.1406 -0.096 -0.3054 -0.9369>}
 dyn_hair_all {parent=head, loc=<4.8199 3.827 -6.7249>, rot=<0.2283 -0.9469 -0.2242 -0.0323>}
 dyn_hair_1_l {parent=dyn_hair_all, loc=<4.9144 0.5221 -0.6906>, rot=<-0.0509 -0.0825 -0.3137 -0.9446>}
 dyn_hair_2_l {parent=dyn_hair_all, loc=<4.607 -0.6869 -0.0089>, rot=<-0.1085 0.0036 0.2565 -0.9604>}
 dyn_head {parent=head, loc=<9.0767 -4.292 0>, rot=<0 0 0 1>}
 dyn_star {parent=neck_01, loc=<2.196 5.139 0>, rot=<0 0 -0.6928 -0.7212>}
 clavicle_l {parent=spine_05, loc=<5.4831 -0.1053 -3.2946>, rot=<0.0465 -0.7289 -0.009 -0.6829>}
 upperarm_l {parent=clavicle_l, loc=<9.8658 -1.2136 -0.4105>, rot=<0.0164 -0.3245 -0.0269 -0.9454>}
 clavicle_r {parent=spine_05, loc=<5.4831 -0.1053 3.2946>, rot=<0.7289 0.0465 -0.6829 0.009>}
 upperarm_r {parent=clavicle_r, loc=<-9.8658 1.2136 0.4105>, rot=<0.0164 -0.3245 -0.0269 -0.9454>}
]
variables [
 Is dew {type=bool, default=False}
 is sky {type=bool, default=False}
]
