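# Physics asset for the AgentSherbert female medium face-accessory mesh, driving secondary motion on its hair strands and two tails.
# Capsules on head, spine_05, clavicles and upper arms form a constrained body frame around the hair.
# Limited swing/twist joints chain a right strand and a shared-root left pair from the head.
# Each tail hangs from a zero-size, angularly locked root capsule, with two limited segments below it.
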PhysicsAsset /BRCosmetics/Characters/Player/Female/Medium/Bodies/F_MED_AgentSherbert/Meshes/Parts/F_MED_AgentSherbert_FaceAcc_Physics
bodies {
 clavicle_l [
  {
   shapeType=Capsule
   length=7
   radius=4.5
   shapeName=clavicle_l_capsule
   center=<7.1692 0.5926 0.1017>
   rotation=<rot 89.187 -179.9975 95.4686>
  }
  {shapeType=Capsule, length=5, radius=7, shapeName=clavicle_l_capsule}
 ]
 clavicle_r [
  {
   shapeType=Capsule
   length=4.5
   radius=5.5
   shapeName=clavicle_l_capsule
   center=<-7.1087 0.1293 0.5012>
   rotation=<rot 89.187 -179.9975 -84.5314>
  }
  {shapeType=Capsule, length=3, radius=8, shapeName=clavicle_l_capsule, rotation=<rot 0 0 -180>}
 ]
 dyn_hair_1_l [{
   shapeType=Capsule
   length=3.9198
   radius=0.8527
   shapeName=dyn_hair_1_l_capsule
   center=<3.4029 0.0994 -0.013>
   rotation=<rot 0.2874 93.0403 -87.524>
  }]
 dyn_hair_2_l [{
   shapeType=Capsule
   length=1.9801
   radius=0.7086
   shapeName=dyn_hair_2_l_capsule
   center=<2.0317 -0.1391 0.1315>
   rotation=<rot 0 78.3988 -85.368>
  }]
 dyn_hair_3_r [{
   shapeType=Capsule
   length=1.1687
   radius=1.5417
   shapeName=dyn_hair_3_r_capsule
   center=<2.0899 0.1728 -0.5619>
   rotation=<rot 0.1664 -90.3045 -87.0622>
  }]
 dyn_hair_3_r_1 [{
   shapeType=Capsule
   length=2.1602
   radius=1.5443
   shapeName=dyn_hair_3_r_1_capsule
   center=<2.0932 -1.0154 0.4401>
   rotation=<rot 0 89.8408 -86.2139>
  }]
 dyn_hair_3_r_2 [{
   shapeType=Capsule
   length=1.191
   radius=2.3254
   shapeName=dyn_hair_3_r_2_capsule
   center=<0.8166 0.4755 0.0126>
   rotation=<rot 0 87.3707 -89.3361>
  }]
 dyn_hair_all [{
   shapeType=Capsule
   length=3.3349
   radius=0.5359
   shapeName=dyn_hair_all_capsule
   center=<2.2389 0.0946 0.8084>
   rotation=<rot -1.6386 -93.0765 -85.4909>
  }]
 dyn_hair_tail_1_l [{shapeType=Capsule, length=0, radius=0}]
 dyn_hair_tail_1_l1 [{shapeType=Capsule, length=20, radius=3.5, shapeName=dyn_hair_tair_1_r_capsule, center=<8.6061 0 0>, rotation=<rot -18.9783 75.4893 -93.7146>}]
 dyn_hair_tail_2_l [{
   shapeType=Capsule
   length=1.5
   radius=3.8329
   shapeName=dyn_hair_tail_2_l_capsule
   center=<5.4013 0.0953 0.0081>
   rotation=<rot 0 93.7423 -89.6806>
  }]
 dyn_hair_tair_1_r [{shapeType=Capsule, length=0, radius=0, rotation=<rot -8.2615 -0.0001 0>}]
 dyn_hair_tair_1_r1 [{shapeType=Capsule, length=17.5, radius=3, shapeName=dyn_hair_tair_1_r_capsule, center=<8.6061 0 0>, rotation=<rot -18.5962 -80.5752 95.3535>}]
 dyn_hair_tair_2_r [{
   shapeType=Capsule
   length=2
   radius=3.3329
   shapeName=dyn_hair_tail_2_l_capsule
   center=<6.3412 0.1567 0.0134>
   rotation=<rot 0 93.7423 -89.6806>
  }]
 head [{shapeType=Capsule, length=6.0755, radius=6.9232, center=<4.035 3.3717 -0.4729>, rotation=<rot 1.7206 -89.316 -88.3374>}]
 spine_05 [{shapeType=Capsule, length=26.5, radius=3.5, shapeName=spine_05_capsule, center=<3.0661 -5.2608 0>}]
 upperarm_l [{shapeType=Capsule, length=5, radius=4, shapeName=upperarm_r_capsule, center=<2.7697 0.6508 -0.3343>, rotation=<rot -0.0001 -89.9999 -90>}]
 upperarm_r [{shapeType=Capsule, length=5, radius=4, shapeName=upperarm_r_capsule, center=<-2.7697 0.6508 0.3343>, rotation=<rot -0.0001 -89.9999 -90>}]
}
constraints [
 {bone1Name=head, bone2Name=spine_05, swing1LimitDegrees=0, swing2LimitDegrees=0, twistLimitDegrees=0}
 {bone1Name=dyn_hair_3_r, bone2Name=head, swing1LimitDegrees=11, swing2LimitDegrees=11, twistLimitDegrees=5}
 {bone1Name=dyn_hair_3_r_1, bone2Name=dyn_hair_3_r, swing1LimitDegrees=40, swing2LimitDegrees=30, twistLimitDegrees=5}
 {bone1Name=dyn_hair_3_r_2, bone2Name=dyn_hair_3_r_1, swing1LimitDegrees=21, swing2LimitDegrees=21, twistLimitDegrees=5}
 {bone1Name=dyn_hair_all, bone2Name=head, swing1LimitDegrees=9, swing2LimitDegrees=4, twistLimitDegrees=1}
 {bone1Name=dyn_hair_1_l, bone2Name=dyn_hair_all, swing1LimitDegrees=11, swing2LimitDegrees=11, twistLimitDegrees=5}
 {bone1Name=dyn_hair_2_l, bone2Name=dyn_hair_all, swing1LimitDegrees=11, swing2LimitDegrees=11, twistLimitDegrees=5}
 {
  bone1Name=dyn_hair_tail_1_l
  bone2Name=head
  swing1LimitDegrees=15
  swing1Motion=Locked
  swing2LimitDegrees=3
  swing2Motion=Locked
  twistLimitDegrees=0.1
  twistMotion=Locked
 }
 {bone1Name=dyn_hair_tail_2_l, bone2Name=dyn_hair_tail_1_l1, swing1LimitDegrees=5, swing2LimitDegrees=35, twistLimitDegrees=0}
 {bone1Name=clavicle_l, bone2Name=spine_05, swing1LimitDegrees=0, swing2LimitDegrees=0, twistLimitDegrees=0}
 {bone1Name=dyn_hair_tail_1_l1, bone2Name=dyn_hair_tail_1_l, swing1LimitDegrees=15, swing2LimitDegrees=0, twistLimitDegrees=0}
 {
  bone1Name=upperarm_l
  bone2Name=clavicle_l
  swing1LimitDegrees=45
  swing1Motion=Locked
  swing2LimitDegrees=45
  swing2Motion=Locked
  twistLimitDegrees=45
  twistMotion=Locked
 }
 {
  bone1Name=upperarm_r
  bone2Name=clavicle_r
  swing1LimitDegrees=45
  swing1Motion=Locked
  swing2LimitDegrees=45
  swing2Motion=Locked
  twistLimitDegrees=45
  twistMotion=Locked
 }
 {
  bone1Name=dyn_hair_tair_1_r
  bone2Name=head
  swing1LimitDegrees=45
  swing1Motion=Locked
  swing2LimitDegrees=45
  swing2Motion=Locked
  twistLimitDegrees=45
  twistMotion=Locked
 }
 {bone1Name=dyn_hair_tair_1_r1, bone2Name=dyn_hair_tair_1_r, swing1LimitDegrees=20, swing2LimitDegrees=0, twistLimitDegrees=0}
 {bone1Name=dyn_hair_tair_2_r, bone2Name=dyn_hair_tair_1_r1, swing1LimitDegrees=5, swing2LimitDegrees=55, twistLimitDegrees=0}
 {bone1Name=clavicle_r, bone2Name=spine_05, swing1LimitDegrees=0, swing2LimitDegrees=0, twistLimitDegrees=0}
]
