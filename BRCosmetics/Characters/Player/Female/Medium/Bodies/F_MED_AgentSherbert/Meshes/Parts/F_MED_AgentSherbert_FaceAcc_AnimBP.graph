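# Animation Blueprint for Agent Sherbert's face accessory, adding head-relative physics that adapts to the Dew body variant.
# It copies the attached parent's pose, layering clavicle and upper-arm branches from the stored body mesh.
# A Control Rig receives Dew and skydiving flags, then feeds a cached pose into two RigidBody simulations with separate physics assets.
# The update event detects Dew by comparing the pawn's part mesh, and that flag selects the simulation output.
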AnimBlueprint /BRCosmetics/Characters/Player/Female/Medium/Bodies/F_MED_AgentSherbert/Meshes/Parts/F_MED_AgentSherbert_FaceAcc_AnimBP
target_skeleton=&/F_MED_AgentSherbert_FaceAcc_Skeleton
graphs [
 AnimGraph {nodes [
   n1 = RigidBody!(
    ComponentPose=Local To Component(LocalPose=Use cached pose 'PreSim'())
    bAlphaBoolEnabled:bool
    ExternalForce="1,1,1"
    ComponentLinearAccScale:Vector="0.3,1,0.7"
    ComponentLinearVelScale:Vector="0.3,1,0.7"
   ) -> (Pose) {
    alphaInputType=Bool
    bClampLinearTranslationLimitToRefPose=true
    bOverrideWorldGravity=true
    cachedBoundsScale=1.32
    overridePhysicsAsset=&/F_MED_AgentSherbert_FaceAcc_Physics_Dew
    simulationSpace=BaseBoneSpace
    alphaBoolBlend {blendOption=Linear}
    baseBoneRef {boneName=head}
    componentAppliedLinearAccClamp=<1200 1200 1200>
    componentLinearAccScale=<1 1 0.5>
    componentLinearVelScale=<1 1 0.5>
    overrideWorldGravity {z=-980}
    simSpaceSettings {
     maxAngularAcceleration=500
     maxAngularVelocity=3
     maxLinearAcceleration=5
     maxLinearVelocity=50
     velocityScaleZ=0
     worldAlpha=1
     externalLinearDragV=<0.1 0.1>
    }
   }
   n2 = RigidBody(
    ComponentPose=Local To Component(LocalPose=Use cached pose 'PreSim'())
    ExternalForce="1,1,1"
    ComponentLinearAccScale="1,1,0.5"
    ComponentLinearVelScale="1,1,0.5"
   ) {
    alphaInputType=Bool
    bClampLinearTranslationLimitToRefPose=true
    bOverrideWorldGravity=true
    cachedBoundsScale=1.32
    overridePhysicsAsset=&/F_MED_AgentSherbert_FaceAcc_Physics
    simulationSpace=BaseBoneSpace
    alphaBoolBlend {blendOption=Linear}
    baseBoneRef {boneName=head}
    componentAppliedLinearAccClamp=<1300 1300 1300>
    componentLinearAccScale=<1 1 0.5>
    componentLinearVelScale=<1 1 0.5>
    overrideWorldGravity {z=-980}
    simSpaceSettings {
     maxAngularAcceleration=1000
     maxAngularVelocity=3
     maxLinearAcceleration=800
     maxLinearVelocity=500
     velocityScaleZ=0
     worldAlpha=1
     externalLinearDragV=<0.1 0.1>
    }
   }
   Output Pose(Result=Blend Poses by bool(BlendPose_0=Component To Local(ComponentPose=@n1), BlendPose_1=Component To Local(ComponentPose=@n2), bActiveValue=Get Dew() -> (Dew:bool)) {blendType=Linear}) {layerGroup=None}
   n3 = Layered blend per bone(
    BasePose=Copy Pose From Mesh() {bUseAttachedParent=true}
    BlendPoses_0=Copy Pose From Mesh(SourceMeshComponent=Get Body mesh() -> (Body mesh:Skeletal Mesh Component Object Reference))
   ) {layerSetup [{branchFilters [{boneName=clavicle_l}, {boneName=clavicle_r}, {boneName=upperarm_l}, {boneName=upperarm_r}]}]}
   n4 = Control Rig!(Source=@n3, Alpha, Is dew:bool=Get Dew(), is sky:bool=Get sky() -> (sky:bool)) -> (Pose) {
    alphaBoolBlend {blendOption=Linear}
    controlRigAssetReference {blueprintRigClass=&/F_MED_AgentSherbert_FaceAcc_CtrlRig.F_MED_AgentSherbert_FaceAcc_CtrlRig_C}
   }
   Save cached pose 'PreSim'(Pose=@n4) {cacheName=PreSim}
  ]}
 EventGraph {nodes [
   Event BlueprintUpdateAnimation()
   > n1 = Sequence()
   n2 = Get As Fort Player Pawn Athena() -> (As Fort Player Pawn Athena:Fort Player Pawn Athena Object Reference)
   Is Valid(exec=@n1.then_1, InputObject=@n2)
   n3 = GetSkeletalMeshForPartType(self=@n2)
   > Set Dew(Dew:bool="Equal (Object)"(
     A=GetSkeletalMeshAsset!(self:Skeletal Mesh Component Object Reference=@n3) -> (ReturnValue:Skeletal Mesh Object Reference)
     B=/BRCosmetics/Characters/Player/Female/Medium/Bodies/F_MED_AgentSherbert_Dew/Meshes/F_MED_AgentSherbert_Dew
    )) -> (Output_Get:bool)
   > Set sky(sky:bool=OR Boolean!(A=Get bIsSkyDiving(), B=Get bIsSkydiveFloating(), C:bool=Get bIsSkydiveDiveMode()) -> (ReturnValue)) -> (Output_Get:bool)
   n4 = Event BlueprintInitializeAnimation()
   n5 = TryGetPawnOwner()
   Is Valid(exec=[@n4.then, @n7.then], InputObject=@n5)
   > n6 = Cast To FortPlayerPawnAthena(Object=@n5)
   > Set As Fort Player Pawn Athena(As Fort Player Pawn Athena:Fort Player Pawn Athena Object Reference=@n6) -> (Output_Get:Fort Player Pawn Athena Object Reference)
   > Set Body mesh(Body mesh:Skeletal Mesh Component Object Reference=GetOwnerSkeletalMeshComponent()) -> (Output_Get:Skeletal Mesh Component Object Reference)
   n7 = Event OnCharacterCustomizationLoaded()
  ]}
]
variables [
 As Fort Player Pawn Athena {type=FortPlayerPawnAthena}
 Body mesh {type=SkeletalMeshComponent}
 Dew {type=boolean}
 sky {type=boolean}
]
